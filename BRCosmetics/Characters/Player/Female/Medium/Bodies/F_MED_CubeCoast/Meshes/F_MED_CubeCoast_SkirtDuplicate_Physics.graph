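# Physics asset for the female medium CubeCoast body's duplicate skirt mesh, rigging skirt panels and right-side waist accessories for physical motion.
# Capsules cover pelvis, thighs and right calf, plus chains of skirt, cord and strap bones.
# Mostly zero-size MSTR bodies root the chains, most joined to pelvis by free constraints.
# Chain links have limited swing and twist, nearly hinge-like on left skirt layers; the strap forks into inner and outer chains.
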PhysicsAsset /BRCosmetics/Characters/Player/Female/Medium/Bodies/F_MED_CubeCoast/Meshes/F_MED_CubeCoast_SkirtDuplicate_Physics
bodies {
 calf_r [
  {shapeType=Capsule, length=12, radius=6, shapeName=calf_r_capsule, center=<12 2 0>, rotation=<rot 90 80.7735 70.7736>}
  {shapeType=Capsule, length=17, radius=5, shapeName=calf_r_capsule, center=<32 2 1>, rotation=<rot 90 0 0>}
 ]
 dyn_01Skirt_bk_01_l [{shapeType=Capsule, length=20, radius=4, shapeName=dyn_01Skirt_bk_01_l_capsule, center=<10 0 0>, rotation=<rot 0 0 90>}]
 dyn_01Skirt_bk_02_l [{shapeType=Capsule, length=15, radius=5, shapeName=dyn_01Skirt_bk_02_l_capsule, center=<8.5 0 0>, rotation=<rot 0 0 90>}]
 dyn_01Skirt_bk_03_l [{shapeType=Capsule, length=20, radius=4, shapeName=dyn_01Skirt_bk_03_l_capsule, center=<10 0 0>, rotation=<rot 0 0 90>}]
 dyn_01Skirt_bk_l_MSTR [{shapeType=Capsule, length=0, radius=0, shapeName=dyn_01Skirt_bk_l_MSTR_capsule, rotation=<rot 0 -90 0>}]
 dyn_01Skirt_ft_01_l [{shapeType=Capsule, length=7.5, radius=1.5, shapeName=dyn_01Skirt_ft_01_l_capsule, center=<10 -10.3406 -3.9775>, rotation=<rot 0.0005 0 50>}]
 dyn_01Skirt_ft_02_l [{shapeType=Capsule, length=7.5, radius=1.5, shapeName=dyn_01Skirt_ft_02_l_capsule, center=<10 -10.9998 -3.192>, rotation=<rot 0.0005 0 50>}]
 dyn_01Skirt_ft_03_l [{shapeType=Capsule, length=7.5, radius=1.5, shapeName=dyn_01Skirt_ft_03_l_capsule, center=<10 -11.6472 -2.4205>, rotation=<rot 0.0005 0 50>}]
 dyn_01Skirt_ft_l_MSTR [{shapeType=Capsule, length=0, radius=0, shapeName=dyn_01Skirt_ft_l_MSTR_capsule}]
 dyn_02Skirt_bk_01_l [{shapeType=Capsule, length=15, radius=2, shapeName=dyn_02Skirt_bk_01_l_capsule, center=<10 0 2>, rotation=<rot 0 0 90>}]
 dyn_02Skirt_bk_02_l [{shapeType=Capsule, length=15, radius=2, shapeName=dyn_02Skirt_bk_02_l_capsule, center=<10 0 2.4>, rotation=<rot 0 0 90>}]
 dyn_02Skirt_bk_03_l [{shapeType=Capsule, length=15, radius=2, shapeName=dyn_02Skirt_bk_03_l_capsule, center=<10 0 2>, rotation=<rot 0 0 90>}]
 dyn_02Skirt_bk_l_MSTR [{shapeType=Capsule, length=0, radius=0, shapeName=dyn_02Skirt_bk_l_MSTR_capsule, rotation=<rot 0 -90 0>}]
 dyn_02Skirt_ft_01_l [
  {
   shapeType=Capsule
   length=15
   radius=1.5
   shapeName=dyn_02Skirt_ft_01_l_capsule
   center=<10 6.3662 -3.3659>
   rotation=<rot 0.0002 0.0002 150>
  }
  {shapeType=Capsule, length=15, radius=1.5, shapeName=dyn_02Skirt_ft_01_l_capsule, center=<10 -9.7989 -2.524>, rotation=<rot 0.0003 0.0003 50>}
 ]
 dyn_02Skirt_ft_02_l [
  {
   shapeType=Capsule
   length=15
   radius=1.5
   shapeName=dyn_02Skirt_ft_02_l_capsule
   center=<10 6.7258 -3.1583>
   rotation=<rot 0.0002 0.0002 150>
  }
  {
   shapeType=Capsule
   length=15
   radius=1.5
   shapeName=dyn_02Skirt_ft_02_l_capsule
   center=<10 -9.9112 -2.3902>
   rotation=<rot 0.0002 0.0002 50>
  }
 ]
 dyn_02Skirt_ft_03_l [
  {
   shapeType=Capsule
   length=15
   radius=1.5
   shapeName=dyn_02Skirt_ft_03_l_capsule
   center=<10 6.2151 -3.4532>
   rotation=<rot 0.0002 0.0002 150>
  }
  {
   shapeType=Capsule
   length=15
   radius=1.5
   shapeName=dyn_02Skirt_ft_03_l_capsule
   center=<10 -10.338 -1.8816>
   rotation=<rot 0.0003 0.0003 50>
  }
 ]
 dyn_02Skirt_ft_l_MSTR [{shapeType=Capsule, length=0, radius=0, shapeName=dyn_02Skirt_ft_l_MSTR_capsule}]
 dyn_Skirt_bk_01_r [{shapeType=Capsule, length=15, radius=4, shapeName=dyn_Skirt_bk_01_r_capsule, center=<10 0 0>, rotation=<rot 0 0 90>}]
 dyn_Skirt_bk_02_r [{shapeType=Capsule, length=15, radius=5, shapeName=dyn_Skirt_bk_02_r_capsule, center=<8.5 0 0>, rotation=<rot 0 0 90>}]
 dyn_Skirt_bk_03_r [{shapeType=Capsule, length=15, radius=4, shapeName=dyn_Skirt_bk_03_r_capsule, center=<10 0 0>, rotation=<rot 0 0 90>}]
 dyn_Skirt_bk_r_MSTR [{shapeType=Capsule, length=0, radius=0, shapeName=dyn_Skirt_bk_r_MSTR_capsule, rotation=<rot 0 -90 0>}]
 dyn_Skirt_ft_01_r [{shapeType=Capsule, length=15, radius=2, shapeName=dyn_Skirt_ft_01_r_capsule, center=<10 0.2 -1.5>, rotation=<rot 0 0 100>}]
 dyn_Skirt_ft_02_r [{shapeType=Capsule, length=15, radius=2, shapeName=dyn_Skirt_ft_02_r_capsule, center=<10 0 0>, rotation=<rot 0 0 100>}]
 dyn_Skirt_ft_r_MSTR [{shapeType=Capsule, length=0, radius=0, shapeName=dyn_Skirt_ft_r_MSTR_capsule, rotation=<rot 0 -90 0>}]
 dyn_WaistAccessory_01_r [{shapeType=Capsule, length=1, radius=0.7, shapeName=dyn_WaistAccessory_01_r_capsule, center=<1 0 0>, rotation=<rot 0 0 90>}]
 dyn_WaistAccessory_02_r [{shapeType=Capsule, length=1, radius=3, shapeName=dyn_WaistAccessory_02_r_capsule, center=<2 0 0.5>, rotation=<rot 0 0 90>}]
 dyn_WaistAccessory_innerstrap_01_r [{shapeType=Capsule, length=15, radius=0.75, shapeName=dyn_WaistAccessory_innerstrap_01_r_capsule, center=<5 0 -0.25>, rotation=<rot 90 0 0>}]
 dyn_WaistAccessory_innerstrap_02_r [{shapeType=Capsule, length=15, radius=0.75, shapeName=dyn_WaistAccessory_innerstrap_02_r_capsule, center=<5 0 -0.25>, rotation=<rot 90 0 0>}]
 dyn_WaistAccessory_innerstrap_03_r [{shapeType=Capsule, length=15, radius=0.75, shapeName=dyn_WaistAccessory_innerstrap_03_r_capsule, center=<5 0 -0.25>, rotation=<rot 90 0 0>}]
 dyn_WaistAccessory_lowercord_01_r [{shapeType=Capsule, length=15, radius=1.5, shapeName=dyn_WaistAccessory_lowercord_01_r_capsule, center=<2 0 2>, rotation=<rot 80 -90.0002 0>}]
 dyn_WaistAccessory_lowercord_02_r [{
   shapeType=Capsule
   length=15
   radius=1.5
   shapeName=dyn_WaistAccessory_lowercord_02_r_capsule
   center=<2 0 2>
   rotation=<rot 90 90 -179.9994>
  }]
 dyn_WaistAccessory_lowercord_03_r [{
   shapeType=Capsule
   length=15
   radius=1.5
   shapeName=dyn_WaistAccessory_lowercord_03_r_capsule
   center=<2 0 2>
   rotation=<rot 90 90 -179.9994>
  }]
 dyn_WaistAccessory_lowercord_04_r [{
   shapeType=Capsule
   length=15
   radius=1.5
   shapeName=dyn_WaistAccessory_lowercord_04_r_capsule
   center=<2 0 2>
   rotation=<rot 90 90 -179.9994>
  }]
 dyn_WaistAccessory_lowercord_05_r [{
   shapeType=Capsule
   length=15
   radius=1.5
   shapeName=dyn_WaistAccessory_lowercord_05_r_capsule
   center=<2 0 2>
   rotation=<rot 90 90 -179.9994>
  }]
 dyn_WaistAccessory_lowercord_r_MSTR [{shapeType=Capsule, length=0, radius=0, shapeName=dyn_WaistAccessory_lowercord_r_MSTR_capsule, center=<5 0 1>, rotation=<rot 90 0 0>}]
 dyn_WaistAccessory_outerstrap_01_r [{shapeType=Capsule, length=15, radius=0.75, shapeName=dyn_WaistAccessory_outerstrap_01_r_capsule, center=<5 0 -0.25>, rotation=<rot 90 0 0>}]
 dyn_WaistAccessory_outerstrap_02_r [{shapeType=Capsule, length=15, radius=0.75, shapeName=dyn_WaistAccessory_outerstrap_02_r_capsule, center=<5 0 -0.25>, rotation=<rot 90 0 0>}]
 dyn_WaistAccessory_outerstrap_03_r [{shapeType=Capsule, length=15, radius=0.75, shapeName=dyn_WaistAccessory_outerstrap_03_r_capsule, center=<5 0 -0.25>, rotation=<rot 90 0 0>}]
 dyn_WaistAccessory_r_MSTR [{shapeType=Capsule, length=0, radius=0, shapeName=dyn_WaistAccessory_r_MSTR_capsule}]
 dyn_WaistAccessory_strap_01_r [{shapeType=Capsule, length=15, radius=0.75, shapeName=dyn_WaistAccessory_strap_01_r_capsule, center=<5 0 -0.25>, rotation=<rot 90 0 0>}]
 dyn_WaistAccessory_strap_02_r [{shapeType=Capsule, length=15, radius=0.75, shapeName=dyn_WaistAccessory_strap_02_r_capsule, center=<5 0 -0.25>, rotation=<rot 90 0 0>}]
 dyn_WaistAccessory_strap_03_r [{shapeType=Capsule, length=15, radius=0.75, shapeName=dyn_WaistAccessory_strap_03_r_capsule, center=<5 0 -0.25>, rotation=<rot 90 0 0>}]
 dyn_WaistAccessory_strap_r_MSTR [{shapeType=Capsule, length=0.505, radius=0.505, shapeName=dyn_WaistAccessory_strap_r_MSTR_capsule, rotation=<rot 0 -90 0>}]
 dyn_WaistAccessory_uppercord_01_r [{shapeType=Capsule, length=15, radius=1.5, shapeName=dyn_WaistAccessory_uppercord_01_r_capsule, center=<2 0 2>, rotation=<rot 80 -90.0002 0>}]
 dyn_WaistAccessory_uppercord_02_r [{
   shapeType=Capsule
   length=15
   radius=1.5
   shapeName=dyn_WaistAccessory_uppercord_02_r_capsule
   center=<2 0 2>
   rotation=<rot 90 90 -179.9994>
  }]
 dyn_WaistAccessory_uppercord_03_r [{
   shapeType=Capsule
   length=15
   radius=1.5
   shapeName=dyn_WaistAccessory_uppercord_03_r_capsule
   center=<2 0 2>
   rotation=<rot 90 90 -179.9994>
  }]
 dyn_WaistAccessory_uppercord_04_r [{
   shapeType=Capsule
   length=15
   radius=1.5
   shapeName=dyn_WaistAccessory_uppercord_04_r_capsule
   center=<2 0 2>
   rotation=<rot 90 90 -179.9994>
  }]
 dyn_WaistAccessory_uppercord_r_MSTR [{shapeType=Capsule, length=0, radius=0, shapeName=dyn_WaistAccessory_uppercord_r_MSTR_capsule, center=<5 0 1>, rotation=<rot 90 0 0>}]
 pelvis [{shapeType=Capsule, length=10, radius=8.5, shapeName=pelvis_capsule, center=<0 -1 0>}]
 thigh_l [
  {shapeType=Capsule, length=15, radius=8, shapeName=thigh_l_capsule, center=<-10 -1 1>, rotation=<rot 0 80 95>}
  {shapeType=Capsule, length=15, radius=7, shapeName=thigh_l_capsule, center=<-30 0 0>, rotation=<rot 90 0 0>}
 ]
 thigh_r [
  {shapeType=Capsule, length=15, radius=7.75, shapeName=thigh_r_capsule, center=<10 1 -1>, rotation=<rot 0 80 95>}
  {shapeType=Capsule, length=15, radius=6, shapeName=thigh_r_capsule, center=<30 0 0>, rotation=<rot 90 0 0>}
 ]
}
constraints [
 {bone1Name=dyn_Skirt_bk_01_r, bone2Name=dyn_Skirt_bk_r_MSTR, swing1LimitDegrees=20, swing2LimitDegrees=20, twistLimitDegrees=2}
 {bone1Name=dyn_Skirt_bk_02_r, bone2Name=dyn_Skirt_bk_01_r, swing1LimitDegrees=20, swing2LimitDegrees=20, twistLimitDegrees=2}
 {bone1Name=dyn_Skirt_bk_03_r, bone2Name=dyn_Skirt_bk_02_r, swing1LimitDegrees=20, swing2LimitDegrees=20, twistLimitDegrees=2}
 {bone1Name=dyn_Skirt_ft_01_r, bone2Name=dyn_Skirt_ft_r_MSTR, swing1LimitDegrees=30, swing2LimitDegrees=30, twistLimitDegrees=2}
 {bone1Name=dyn_Skirt_ft_02_r, bone2Name=dyn_Skirt_ft_01_r, swing1LimitDegrees=30, swing2LimitDegrees=30, twistLimitDegrees=2}
 {
  bone1Name=dyn_Skirt_bk_r_MSTR
  bone2Name=pelvis
  swing1LimitDegrees=45
  swing1Motion=Free
  swing2LimitDegrees=45
  swing2Motion=Free
  twistLimitDegrees=45
  twistMotion=Free
 }
 {
  bone1Name=dyn_Skirt_ft_r_MSTR
  bone2Name=pelvis
  swing1LimitDegrees=45
  swing1Motion=Free
  swing2LimitDegrees=45
  swing2Motion=Free
  twistLimitDegrees=45
  twistMotion=Free
 }
 {
  bone1Name=thigh_l
  bone2Name=pelvis
  swing1LimitDegrees=45
  swing1Motion=Free
  swing2LimitDegrees=45
  swing2Motion=Free
  twistLimitDegrees=45
  twistMotion=Free
 }
 {
  bone1Name=thigh_r
  bone2Name=pelvis
  swing1LimitDegrees=45
  swing1Motion=Free
  swing2LimitDegrees=45
  swing2Motion=Free
  twistLimitDegrees=45
  twistMotion=Free
 }
 {
  bone1Name=dyn_01Skirt_bk_l_MSTR
  bone2Name=pelvis
  swing1LimitDegrees=45
  swing1Motion=Free
  swing2LimitDegrees=45
  swing2Motion=Free
  twistLimitDegrees=45
  twistMotion=Free
 }
 {bone1Name=dyn_01Skirt_bk_01_l, bone2Name=dyn_01Skirt_bk_l_MSTR, swing1LimitDegrees=2, swing2LimitDegrees=20, twistLimitDegrees=2}
 {bone1Name=dyn_01Skirt_bk_02_l, bone2Name=dyn_01Skirt_bk_01_l, swing1LimitDegrees=2, swing2LimitDegrees=20, twistLimitDegrees=2}
 {bone1Name=dyn_01Skirt_bk_03_l, bone2Name=dyn_01Skirt_bk_02_l, swing1LimitDegrees=2, swing2LimitDegrees=20, twistLimitDegrees=2}
 {
  bone1Name=dyn_01Skirt_ft_l_MSTR
  bone2Name=pelvis
  swing1LimitDegrees=45
  swing1Motion=Free
  swing2LimitDegrees=45
  swing2Motion=Free
  twistLimitDegrees=45
  twistMotion=Free
 }
 {bone1Name=dyn_01Skirt_ft_01_l, bone2Name=dyn_01Skirt_ft_l_MSTR, swing1LimitDegrees=1, swing2LimitDegrees=30, twistLimitDegrees=1}
 {bone1Name=dyn_01Skirt_ft_02_l, bone2Name=dyn_01Skirt_ft_01_l, swing1LimitDegrees=1, swing2LimitDegrees=30, twistLimitDegrees=1}
 {bone1Name=dyn_01Skirt_ft_03_l, bone2Name=dyn_01Skirt_ft_02_l, swing1LimitDegrees=1, swing2LimitDegrees=30, twistLimitDegrees=1}
 {
  bone1Name=dyn_02Skirt_bk_l_MSTR
  bone2Name=pelvis
  swing1LimitDegrees=45
  swing1Motion=Free
  swing2LimitDegrees=45
  swing2Motion=Free
  twistLimitDegrees=45
  twistMotion=Free
 }
 {bone1Name=dyn_02Skirt_bk_01_l, bone2Name=dyn_02Skirt_bk_l_MSTR, swing1LimitDegrees=1, swing2LimitDegrees=10, twistLimitDegrees=1}
 {bone1Name=dyn_02Skirt_bk_02_l, bone2Name=dyn_02Skirt_bk_01_l, swing1LimitDegrees=1, swing2LimitDegrees=10, twistLimitDegrees=1}
 {bone1Name=dyn_02Skirt_bk_03_l, bone2Name=dyn_02Skirt_bk_02_l, swing1LimitDegrees=1, swing2LimitDegrees=10, twistLimitDegrees=1}
 {
  bone1Name=dyn_02Skirt_ft_l_MSTR
  bone2Name=pelvis
  swing1LimitDegrees=45
  swing1Motion=Free
  swing2LimitDegrees=45
  swing2Motion=Free
  twistLimitDegrees=45
  twistMotion=Free
 }
 {bone1Name=dyn_02Skirt_ft_01_l, bone2Name=dyn_02Skirt_ft_l_MSTR, swing1LimitDegrees=1, swing2LimitDegrees=30, twistLimitDegrees=1}
 {bone1Name=dyn_02Skirt_ft_02_l, bone2Name=dyn_02Skirt_ft_01_l, swing1LimitDegrees=1, swing2LimitDegrees=30, twistLimitDegrees=1}
 {bone1Name=dyn_02Skirt_ft_03_l, bone2Name=dyn_02Skirt_ft_02_l, swing1LimitDegrees=1, swing2LimitDegrees=30, twistLimitDegrees=1}
 {
  bone1Name=dyn_WaistAccessory_uppercord_04_r
  bone2Name=dyn_WaistAccessory_uppercord_03_r
  swing1LimitDegrees=15
  swing2LimitDegrees=35
  twistLimitDegrees=15
 }
 {
  bone1Name=dyn_WaistAccessory_lowercord_02_r
  bone2Name=dyn_WaistAccessory_lowercord_01_r
  swing1LimitDegrees=15
  swing2LimitDegrees=35
  twistLimitDegrees=15
 }
 {
  bone1Name=dyn_WaistAccessory_lowercord_03_r
  bone2Name=dyn_WaistAccessory_lowercord_02_r
  swing1LimitDegrees=15
  swing2LimitDegrees=35
  twistLimitDegrees=15
 }
 {
  bone1Name=dyn_WaistAccessory_lowercord_04_r
  bone2Name=dyn_WaistAccessory_lowercord_03_r
  swing1LimitDegrees=15
  swing2LimitDegrees=35
  twistLimitDegrees=15
 }
 {
  bone1Name=dyn_WaistAccessory_lowercord_05_r
  bone2Name=dyn_WaistAccessory_lowercord_04_r
  swing1LimitDegrees=15
  swing2LimitDegrees=35
  twistLimitDegrees=15
 }
 {
  bone1Name=dyn_WaistAccessory_strap_01_r
  bone2Name=dyn_WaistAccessory_strap_r_MSTR
  swing1LimitDegrees=35
  swing2LimitDegrees=35
  twistLimitDegrees=15
 }
 {
  bone1Name=dyn_WaistAccessory_strap_02_r
  bone2Name=dyn_WaistAccessory_strap_01_r
  swing1LimitDegrees=35
  swing2LimitDegrees=35
  twistLimitDegrees=15
 }
 {
  bone1Name=dyn_WaistAccessory_strap_03_r
  bone2Name=dyn_WaistAccessory_strap_02_r
  swing1LimitDegrees=35
  swing2LimitDegrees=35
  twistLimitDegrees=15
 }
 {
  bone1Name=dyn_WaistAccessory_innerstrap_01_r
  bone2Name=dyn_WaistAccessory_strap_03_r
  swing1LimitDegrees=15
  swing2LimitDegrees=15
  twistLimitDegrees=2
 }
 {
  bone1Name=dyn_WaistAccessory_outerstrap_01_r
  bone2Name=dyn_WaistAccessory_strap_03_r
  swing1LimitDegrees=15
  swing2LimitDegrees=15
  twistLimitDegrees=2
 }
 {
  bone1Name=dyn_WaistAccessory_innerstrap_02_r
  bone2Name=dyn_WaistAccessory_innerstrap_01_r
  swing1LimitDegrees=15
  swing2LimitDegrees=15
  twistLimitDegrees=2
 }
 {
  bone1Name=dyn_WaistAccessory_innerstrap_03_r
  bone2Name=dyn_WaistAccessory_innerstrap_02_r
  swing1LimitDegrees=15
  swing2LimitDegrees=15
  twistLimitDegrees=2
 }
 {
  bone1Name=dyn_WaistAccessory_outerstrap_02_r
  bone2Name=dyn_WaistAccessory_outerstrap_01_r
  swing1LimitDegrees=15
  swing2LimitDegrees=15
  twistLimitDegrees=2
 }
 {
  bone1Name=dyn_WaistAccessory_outerstrap_03_r
  bone2Name=dyn_WaistAccessory_outerstrap_02_r
  swing1LimitDegrees=15
  swing2LimitDegrees=15
  twistLimitDegrees=2
 }
 {bone1Name=dyn_WaistAccessory_02_r, bone2Name=dyn_WaistAccessory_01_r, swing1LimitDegrees=35, swing2LimitDegrees=35, twistLimitDegrees=2}
 {
  bone1Name=dyn_WaistAccessory_uppercord_02_r
  bone2Name=dyn_WaistAccessory_uppercord_01_r
  swing1LimitDegrees=15
  swing2LimitDegrees=35
  twistLimitDegrees=15
 }
 {
  bone1Name=dyn_WaistAccessory_uppercord_03_r
  bone2Name=dyn_WaistAccessory_uppercord_02_r
  swing1LimitDegrees=15
  swing2LimitDegrees=35
  twistLimitDegrees=15
 }
 {
  bone1Name=calf_r
  bone2Name=thigh_r
  swing1LimitDegrees=45
  swing1Motion=Free
  swing2LimitDegrees=45
  swing2Motion=Free
  twistLimitDegrees=45
  twistMotion=Free
 }
 {
  bone1Name=dyn_WaistAccessory_r_MSTR
  bone2Name=pelvis
  swing1LimitDegrees=45
  swing1Motion=Free
  swing2LimitDegrees=45
  swing2Motion=Free
  twistLimitDegrees=45
  twistMotion=Free
 }
 {bone1Name=dyn_WaistAccessory_01_r, bone2Name=dyn_WaistAccessory_r_MSTR, swing1LimitDegrees=5, swing2LimitDegrees=35, twistLimitDegrees=2}
 {
  bone1Name=dyn_WaistAccessory_uppercord_r_MSTR
  bone2Name=dyn_WaistAccessory_r_MSTR
  swing1LimitDegrees=2
  swing2LimitDegrees=45
  twistLimitDegrees=2
 }
 {
  bone1Name=dyn_WaistAccessory_lowercord_r_MSTR
  bone2Name=dyn_WaistAccessory_r_MSTR
  swing1LimitDegrees=2
  swing2LimitDegrees=45
  twistLimitDegrees=2
 }
 {
  bone1Name=dyn_WaistAccessory_lowercord_01_r
  bone2Name=dyn_WaistAccessory_lowercord_r_MSTR
  swing1LimitDegrees=35
  swing2LimitDegrees=35
  twistLimitDegrees=35
 }
 {
  bone1Name=dyn_WaistAccessory_uppercord_01_r
  bone2Name=dyn_WaistAccessory_uppercord_r_MSTR
  swing1LimitDegrees=35
  swing2LimitDegrees=35
  twistLimitDegrees=35
 }
 {
  bone1Name=dyn_WaistAccessory_strap_r_MSTR
  bone2Name=pelvis
  swing1LimitDegrees=45
  swing1Motion=Free
  swing2LimitDegrees=45
  swing2Motion=Free
  twistLimitDegrees=45
  twistMotion=Free
 }
]
