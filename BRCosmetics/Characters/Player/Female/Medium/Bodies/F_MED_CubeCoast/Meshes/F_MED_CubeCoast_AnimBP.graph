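# Animation Blueprint for the F_MED_CubeCoast female body cosmetic, driving its skirt and waist accessories from the attached parent's copied pose.
# A deform Control Rig, thigh/pelvis constraints and a skirt Control Rig feed a pelvis-space RigidBody simulation with overridden gravity.
# Movement states multiply into SkirtAlpha to weaken physics.
# Additive bone offsets correct crouching before simulation, then jumping/falling, skydiving, front-end and curve-detected custom idle poses afterwards.
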AnimBlueprint /BRCosmetics/Characters/Player/Female/Medium/Bodies/F_MED_CubeCoast/Meshes/F_MED_CubeCoast_AnimBP
target_skeleton=&/F_MED_CubeCoast_Skeleton
graphs [
 AnimGraph {nodes [
   n1 = Control Rig(Source=Copy Pose From Mesh() {bUseAttachedParent=true}) {
    alphaBoolBlend {blendOption=Linear}
    controlRigAssetReference {blueprintRigClass=&/F_MED_CubeCoast_DeformRig_ControlRig.F_MED_CubeCoast_DeformRig_ControlRig_C}
   }
   n2 = "Constraint - dyn_01Skirt_ft_l_MSTR (RR)"!(ComponentPose=Local To Component(LocalPose=@n1), Alpha=1, ConstraintWeights_0=0.7, ConstraintWeights_1:float=0.3) -> (Pose) {
    alphaBoolBlend {blendOption=Linear}
    boneToModify {boneName=dyn_01Skirt_ft_l_MSTR}
    constraintSetup [{transformType=Rotation, targetBone {boneName=thigh_l}}, {transformType=Rotation, targetBone {boneName=pelvis}}]
    constraintWeights [1,1]
   }
   n3 = "Constraint - dyn_02Skirt_ft_l_MSTR (RR)"!(ComponentPose=@n2, Alpha=1, ConstraintWeights_0=0.9, ConstraintWeights_1:float=0.1) -> (Pose) {
    alphaBoolBlend {blendOption=Linear}
    boneToModify {boneName=dyn_02Skirt_ft_l_MSTR}
    constraintSetup [{transformType=Rotation, targetBone {boneName=thigh_l}}, {transformType=Rotation, targetBone {boneName=pelvis}}]
    constraintWeights [1,1]
   }
   n4 = "Constraint - dyn_02Skirt_bk_l_MSTR (RR)"!(ComponentPose=@n3, Alpha=1, ConstraintWeights_0=0.7, ConstraintWeights_1:float=0.3) -> (Pose) {
    alphaBoolBlend {blendOption=Linear}
    boneToModify {boneName=dyn_02Skirt_bk_l_MSTR}
    constraintSetup [{transformType=Rotation, targetBone {boneName=thigh_l}}, {transformType=Rotation, targetBone {boneName=pelvis}}]
    constraintWeights [1,1]
   }
   n5 = "Constraint - dyn_Skirt_ft_r_MSTR (RR)"!(ComponentPose=@n4, Alpha=1, ConstraintWeights_0=0.7, ConstraintWeights_1:float=0.3) -> (Pose) {
    alphaBoolBlend {blendOption=Linear}
    boneToModify {boneName=dyn_Skirt_ft_r_MSTR}
    constraintSetup [{transformType=Rotation, targetBone {boneName=thigh_r}}, {transformType=Rotation, targetBone {boneName=pelvis}}]
    constraintWeights [1,1]
   }
   n6 = "Constraint - dyn_WaistAccessory_r (RR)"!(ComponentPose=@n5, Alpha=1, ConstraintWeights_0=0.3, ConstraintWeights_1:float=0.7) -> (Pose) {
    alphaBoolBlend {blendOption=Linear}
    boneToModify {boneName=dyn_WaistAccessory_r}
    constraintSetup [{transformType=Rotation, targetBone {boneName=thigh_r}}, {transformType=Rotation, targetBone {boneName=pelvis}}]
    constraintWeights [1,1]
   }
   n7 = "Constraint - dyn_WaistAccessory_ft_MSTR (RR)"!(ComponentPose=@n6, Alpha=1, ConstraintWeights_0=0.7, ConstraintWeights_1:float=0.3) -> (Pose) {
    alphaBoolBlend {blendOption=Linear}
    boneToModify {boneName=dyn_WaistAccessory_ft_MSTR}
    constraintSetup [{transformType=Rotation, targetBone {boneName=thigh_r}}, {transformType=Rotation, targetBone {boneName=pelvis}}]
    constraintWeights [1,1]
   }
   n8 = "Constraint - dyn_WaistAccessory_bk_MSTR (RR)"!(ComponentPose=@n7, Alpha=1, ConstraintWeights_0=0, ConstraintWeights_1:float=1) -> (Pose) {
    alphaBoolBlend {blendOption=Linear}
    boneToModify {boneName=dyn_WaistAccessory_bk_MSTR}
    constraintSetup [{transformType=Rotation, targetBone {boneName=thigh_r}}, {transformType=Rotation, targetBone {boneName=pelvis}}]
    constraintWeights [1,1]
   }
   n9 = "Transform (Modify) Bone - Bone: dyn_Skirt_bk_r_MSTR"(ComponentPose=@n8, Alpha=1, Translation="1,0,0", Rotation="0,0,0", Scale="1,1,1") {
    rotationMode=BMM_Additive
    rotationSpace=BCS_BoneSpace
    translationMode=BMM_Additive
    translationSpace=BCS_BoneSpace
    alphaBoolBlend {blendOption=Linear}
    boneToModify {boneName=dyn_Skirt_bk_r_MSTR}
    translation {x=1}
   }
   n10 = "Transform (Modify) Bone - Bone: dyn_02Skirt_bk_l_MSTR"(ComponentPose=@n9, Alpha=1, Translation="1,0,0", Rotation="0,0,0", Scale="1,1,1") {
    rotationMode=BMM_Additive
    rotationSpace=BCS_BoneSpace
    translationMode=BMM_Additive
    translationSpace=BCS_BoneSpace
    alphaBoolBlend {blendOption=Linear}
    boneToModify {boneName=dyn_02Skirt_bk_l_MSTR}
    translation {x=1}
   }
   n11 = "Transform (Modify) Bone - Bone: dyn_01Skirt_bk_l_MSTR"(ComponentPose=@n10, Alpha=1, Translation="1,0,0", Rotation="0,0,0", Scale="1,1,1") {
    rotationMode=BMM_Additive
    rotationSpace=BCS_BoneSpace
    translationMode=BMM_Additive
    translationSpace=BCS_BoneSpace
    alphaBoolBlend {blendOption=Linear}
    boneToModify {boneName=dyn_01Skirt_bk_l_MSTR}
    translation {x=1}
   }
   n12 = "Constraint - pelvis_Anchor (P)"(ComponentPose=@n11, Alpha=1, ConstraintWeights_0=1) {
    alphaBoolBlend {blendOption=Linear}
    boneToModify {boneName=pelvis_Anchor}
    constraintSetup [{targetBone {boneName=pelvis}}]
    constraintWeights [1]
   }
   n13 = "Transform (Modify) Bone - Bone: pelvis_Anchor"(ComponentPose=@n12, Alpha=Get Is Crouching Float() -> (Is Crouching Float:float), Translation="1.5,0,0", Rotation="0,0,0", Scale="1,1,1") {
    rotationMode=BMM_Additive
    rotationSpace=BCS_BoneSpace
    translationMode=BMM_Additive
    translationSpace=BCS_BoneSpace
    alphaBoolBlend {blendOption=Linear}
    alphaScaleBiasClamp {bInterpResult=true, interpSpeedDecreasing=3, interpSpeedIncreasing=3}
    boneToModify {boneName=pelvis_Anchor}
    translation {x=1.5}
   }
   n14 = "Constraint - dyn_WaistAccessory_r_MSTR (P)"(ComponentPose=@n13, Alpha=1, ConstraintWeights_0=1) {
    alphaBoolBlend {blendOption=Linear}
    boneToModify {boneName=dyn_WaistAccessory_r_MSTR}
    constraintSetup [{targetBone {boneName=pelvis_Anchor}}]
    constraintWeights [1]
   }
   n15 = Control Rig(Source=Component To Local(ComponentPose=@n14)) {
    alphaBoolBlend {blendOption=Linear}
    alphaScaleBiasClamp {interpSpeedIncreasing=2}
    controlRigAssetReference {blueprintRigClass=&/F_MED_CubeCoast_Skirt_ControlRig.F_MED_CubeCoast_Skirt_ControlRig_C}
   }
   n16 = RigidBody!(
    ComponentPose=Local To Component(LocalPose=@n15)
    Alpha=Get SkirtAlpha() -> (SkirtAlpha:double)
    OverrideWorldGravity:Vector=Get GravityOverride()
    ExternalForce
   ) -> (Pose) {
    bClampLinearTranslationLimitToRefPose=true
    bOverrideWorldGravity=true
    overridePhysicsAsset=&/F_MED_CubeCoast_SkirtDuplicate_Physics
    simulationSpace=BaseBoneSpace
    alphaBoolBlend {blendOption=Linear}
    alphaScaleBiasClamp {interpSpeedIncreasing=2}
    baseBoneRef {boneName=pelvis}
    componentAppliedLinearAccClamp=<3000 3000 3000>
    componentLinearAccScale=<0.4 0.3 0.3>
    componentLinearVelScale=<0.4 0.3 0.3>
    simSpaceSettings {
     maxAngularAcceleration=1000
     maxAngularVelocity=3
     maxLinearAcceleration=700
     maxLinearVelocity=500
     velocityScaleZ=0
     worldAlpha=0.25
     externalAngularVelocity {z=0.1}
     externalLinearDragV=<0.2 0.2>
    }
   }
   n17 = Get Is Jumping() -> (Is Jumping:float)
   n18 = "Transform (Modify) Bone - Bone: pelvis_Anchor"(ComponentPose=@n16, Alpha=@n17, Translation="5,1,0", Rotation="0,0,0", Scale="1,1,1") {
    rotationMode=BMM_Additive
    rotationSpace=BCS_BoneSpace
    translationMode=BMM_Additive
    translationSpace=BCS_BoneSpace
    alphaBoolBlend {blendOption=Linear}
    alphaScaleBiasClamp {bInterpResult=true, interpSpeedIncreasing=2}
    boneToModify {boneName=pelvis_Anchor}
    translation=<5 1>
   }
   n19 = "Constraint - dyn_WaistAccessory_r_MSTR (P)"(ComponentPose=@n18, Alpha=@n17, ConstraintWeights_0=1) {
    alphaBoolBlend {blendOption=Linear}
    boneToModify {boneName=dyn_WaistAccessory_r_MSTR}
    constraintSetup [{targetBone {boneName=pelvis_Anchor}}]
    constraintWeights [1]
   }
   n20 = "Transform (Modify) Bone - Bone: dyn_Skirt_ft_r_MSTR"(ComponentPose=@n19, Alpha=@n17, Translation="0,0,0", Rotation="-30,0,0", Scale="1,1,1") {
    rotationMode=BMM_Additive
    rotationSpace=BCS_BoneSpace
    translationMode=BMM_Additive
    translationSpace=BCS_BoneSpace
    alphaBoolBlend {blendOption=Linear}
    alphaScaleBiasClamp {bInterpResult=true, interpSpeedIncreasing=4}
    boneToModify {boneName=dyn_Skirt_ft_r_MSTR}
    rotation {pitch=-30}
   }
   n21 = "Transform (Modify) Bone - Bone: dyn_WaistAccessory_ft_MSTR"(ComponentPose=@n20, Alpha=@n17, Translation="1,0,-2", Rotation="-20,0,0", Scale="1,1,1") {
    rotationMode=BMM_Additive
    rotationSpace=BCS_BoneSpace
    translationMode=BMM_Additive
    translationSpace=BCS_BoneSpace
    alphaBoolBlend {blendOption=Linear}
    alphaScaleBiasClamp {bInterpResult=true, interpSpeedIncreasing=4}
    boneToModify {boneName=dyn_WaistAccessory_ft_MSTR}
    rotation {pitch=-20}
    translation {x=1, z=-2}
   }
   n22 = "Transform (Modify) Bone - Bone: dyn_Skirt_bk_r_MSTR"(ComponentPose=@n21, Alpha=@n17, Translation="1,0,0", Rotation="30,0,0", Scale="1,1,1") {
    rotationMode=BMM_Additive
    rotationSpace=BCS_BoneSpace
    translationMode=BMM_Additive
    translationSpace=BCS_BoneSpace
    alphaBoolBlend {blendOption=Linear}
    alphaScaleBiasClamp {bInterpResult=true, interpSpeedDecreasing=2, interpSpeedIncreasing=2}
    boneToModify {boneName=dyn_Skirt_bk_r_MSTR}
    rotation {pitch=30}
    translation {x=1}
   }
   n23 = "Transform (Modify) Bone - Bone: dyn_WaistAccessory_strap_01_r"(ComponentPose=@n22, Alpha=@n17, Translation="0,-1,0", Rotation="-7,-5,0", Scale="1,1,1") {
    rotationMode=BMM_Additive
    rotationSpace=BCS_BoneSpace
    translationMode=BMM_Additive
    translationSpace=BCS_BoneSpace
    alphaBoolBlend {blendOption=Linear}
    alphaScaleBiasClamp {bInterpResult=true, interpSpeedIncreasing=4}
    boneToModify {boneName=dyn_WaistAccessory_strap_01_r}
    rotation {pitch=-7, yaw=-5}
    translation {y=-1}
   }
   n24 = "Transform (Modify) Bone - Bone: dyn_WaistAccessory_r_MSTR"(ComponentPose=@n23, Alpha=@n17, Translation="1,1,0", Rotation="0,0,0", Scale="1,1,1") {
    rotationMode=BMM_Additive
    rotationSpace=BCS_BoneSpace
    translationMode=BMM_Additive
    translationSpace=BCS_BoneSpace
    alphaBoolBlend {blendOption=Linear}
    alphaScaleBiasClamp {bInterpResult=true, interpSpeedIncreasing=4}
    boneToModify {boneName=dyn_WaistAccessory_r_MSTR}
    translation=<1 1>
   }
   n25 = "Transform (Modify) Bone - Bone: dyn_WaistAccessory_bk_MSTR"(ComponentPose=@n24, Alpha=@n17, Translation="0,0,0", Rotation="0,0,30", Scale="1,1,1") {
    rotationMode=BMM_Additive
    rotationSpace=BCS_BoneSpace
    translationMode=BMM_Additive
    translationSpace=BCS_BoneSpace
    alphaBoolBlend {blendOption=Linear}
    alphaScaleBiasClamp {bInterpResult=true, interpSpeedIncreasing=4}
    boneToModify {boneName=dyn_WaistAccessory_bk_MSTR}
    rotation {roll=30}
   }
   n26 = "Transform (Modify) Bone - Bone: dyn_WaistAccessory_strap_01_r"(ComponentPose=@n25, Alpha=Get SkyDiving_Float() -> (SkyDiving_Float:float), Translation="0,0,0", Rotation="-7,0,0", Scale="1,1,1") {
    rotationMode=BMM_Additive
    rotationSpace=BCS_BoneSpace
    translationMode=BMM_Additive
    translationSpace=BCS_BoneSpace
    alphaBoolBlend {blendOption=Linear}
    alphaScaleBiasClamp {bInterpResult=true, interpSpeedIncreasing=4}
    boneToModify {boneName=dyn_WaistAccessory_strap_01_r}
    rotation {pitch=-7}
   }
   n27 = Get Is In Front End Alpha() -> (Is In Front End Alpha:double)
   n28 = "Transform (Modify) Bone - Bone: dyn_02Skirt_bk_l_MSTR"(ComponentPose=@n26, Alpha=@n27, Translation="1,2,0", Rotation="0,0,0", Scale="1,1,1") {
    rotationMode=BMM_Additive
    rotationSpace=BCS_BoneSpace
    translationMode=BMM_Additive
    translationSpace=BCS_BoneSpace
    alphaBoolBlend {blendOption=Linear}
    alphaScaleBiasClamp {bInterpResult=true, interpSpeedIncreasing=4}
    boneToModify {boneName=dyn_02Skirt_bk_l_MSTR}
    translation=<1 2>
   }
   n29 = "Transform (Modify) Bone - Bone: dyn_02Skirt_ft_l_MSTR"(ComponentPose=@n28, Alpha=@n27, Translation="2,0,0", Rotation="0,0,0", Scale="1,1,1") {
    rotationMode=BMM_Additive
    rotationSpace=BCS_BoneSpace
    translationMode=BMM_Additive
    translationSpace=BCS_BoneSpace
    alphaBoolBlend {blendOption=Linear}
    alphaScaleBiasClamp {bInterpResult=true, interpSpeedIncreasing=4}
    boneToModify {boneName=dyn_02Skirt_ft_l_MSTR}
    translation {x=2}
   }
   n30 = "Transform (Modify) Bone - Bone: dyn_01Skirt_ft_l_MSTR"(ComponentPose=@n29, Alpha=@n27, Translation="0,0,0", Rotation="-5,0,0", Scale="1,1,1") {
    rotationMode=BMM_Additive
    rotationSpace=BCS_BoneSpace
    translationMode=BMM_Additive
    translationSpace=BCS_BoneSpace
    alphaBoolBlend {blendOption=Linear}
    alphaScaleBiasClamp {bInterpResult=true, interpSpeedIncreasing=4}
    boneToModify {boneName=dyn_01Skirt_ft_l_MSTR}
    rotation {pitch=-5}
   }
   n31 = Get CustomIdle is Playing Float() -> (CustomIdle is Playing Float:float)
   n32 = "Transform (Modify) Bone - Bone: dyn_02Skirt_bk_l_MSTR"(ComponentPose=@n30, Alpha=@n31, Translation="1,2,0", Rotation="0,0,0", Scale="1,1,1") {
    rotationMode=BMM_Additive
    rotationSpace=BCS_BoneSpace
    translationMode=BMM_Additive
    translationSpace=BCS_BoneSpace
    alphaBoolBlend {blendOption=Linear}
    alphaScaleBiasClamp {bInterpResult=true, interpSpeedIncreasing=4}
    boneToModify {boneName=dyn_02Skirt_bk_l_MSTR}
    translation=<1 2>
   }
   n33 = "Transform (Modify) Bone - Bone: dyn_02Skirt_ft_l_MSTR"(ComponentPose=@n32, Alpha=@n31, Translation="2,0,0", Rotation="0,0,0", Scale="1,1,1") {
    rotationMode=BMM_Additive
    rotationSpace=BCS_BoneSpace
    translationMode=BMM_Additive
    translationSpace=BCS_BoneSpace
    alphaBoolBlend {blendOption=Linear}
    alphaScaleBiasClamp {bInterpResult=true, interpSpeedIncreasing=4}
    boneToModify {boneName=dyn_02Skirt_ft_l_MSTR}
    translation {x=2}
   }
   n34 = "Transform (Modify) Bone - Bone: dyn_01Skirt_ft_l_MSTR"(ComponentPose=@n33, Alpha=@n31, Translation="0,0,0", Rotation="-5,0,0", Scale="1,1,1") {
    rotationMode=BMM_Additive
    rotationSpace=BCS_BoneSpace
    translationMode=BMM_Additive
    translationSpace=BCS_BoneSpace
    alphaBoolBlend {blendOption=Linear}
    alphaScaleBiasClamp {bInterpResult=true, interpSpeedIncreasing=4}
    boneToModify {boneName=dyn_01Skirt_ft_l_MSTR}
    rotation {pitch=-5}
   }
   Output Pose(Result=Component To Local(ComponentPose=@n34)) {layerGroup=None}
  ]}
 EventGraph {nodes [
   Event BlueprintUpdateAnimation()
   > n1 = Sequence!(execute) -> (then_0, then_1, then_2:Exec, then_3:Exec, then_4:Exec, then_5:Exec, then_6:Exec, then_7:Exec)
   > n2 = Gravity Override State Function DEPRECATED(CharacterPartAnimInstance=Self-Reference(), GameTimeInSeconds=GetGameTimeInSeconds(), GravityOverrideParams=Get Gravity Override Params())
   > Set GravityOverride(GravityOverride=@n2)
   Set SkirtAlpha(
    execute=@n1.then_1
    SkirtAlpha:double=float * float!(
     A=Get Is Sky Diving Skirt Alpha() -> (Is Sky Diving Skirt Alpha:double)
     B=Get Is Tactical Sprinting Alpha() -> (Is Tactical Sprinting Alpha:double)
     C:float=Get Is Sliding Alpha() -> (Is Sliding Alpha:float)
     D:double=Get Jumping Skirt Alpha() -> (Jumping Skirt Alpha:double)
     E:double=Get Is Swim Sprinting Alpha() -> (Is Swim Sprinting Alpha:double)
     F:double=Get Is Surface Swimming Alpha() -> (Is Surface Swimming Alpha:double)
     G:double=1
     H:float=Get Is Crouching Alpha() -> (Is Crouching Alpha:float)
    ) -> (ReturnValue) {is=title K2Node_PromotableOperator}
   ) -> (Output_Get:double)
   n3 = Sequence!(execute=@n1.then_2) -> (then_0, then_1, then_2, then_3, then_4, then_5, then_6)
   > n4 = Branch(Condition=Get bIsSkyDiving())
   > Set Is Sky Diving Skirt Alpha(Is Sky Diving Skirt Alpha:double=0.5) -> (Output_Get:double)
   Set Is Sky Diving Skirt Alpha(execute=@n4.else, Is Sky Diving Skirt Alpha=1)
   n5 = Branch(execute=@n3.then_1, Condition=Get bIsTacticalSprinting())
   > Set Is Tactical Sprinting Alpha(Is Tactical Sprinting Alpha:double=0.5) -> (Output_Get:double)
   Set Is Tactical Sprinting Alpha(execute=@n5.else, Is Tactical Sprinting Alpha=1)
   n6 = Branch(execute=@n3.then_2, Condition=Get bIsSliding())
   > Set Is Sliding Alpha(Is Sliding Alpha:float=0.25) -> (Output_Get:float)
   Set Is Sliding Alpha(execute=@n6.else, Is Sliding Alpha=1)
   n7 = NOT Boolean(A=Get bIsSkyDiving())
   n8 = Get bIsJumping()
   n9 = NOT Boolean(A=Get bIsSkyDiving())
   n10 = Get bIsFalling()
   n11 = Branch(execute=@n3.then_3, Condition=OR Boolean(A=AND Boolean(A=@n8, B=@n7), B=AND Boolean(A=@n10, B=@n9)))
   > Set Jumping Skirt Alpha(Jumping Skirt Alpha:double=0.5) -> (Output_Get:double)
   Set Jumping Skirt Alpha(execute=@n11.else, Jumping Skirt Alpha=1)
   n12 = Branch(execute=@n3.then_4, Condition=Get bIsSurfaceSwimming())
   > Set Is Surface Swimming Alpha(Is Surface Swimming Alpha:double=0.25) -> (Output_Get:double)
   Set Is Surface Swimming Alpha(execute=@n12.else, Is Surface Swimming Alpha=1)
   n13 = Branch(execute=@n3.then_5, Condition=Get bIsSwimSprinting() -> (bIsSwimSprinting:bool))
   > Set Is Swim Sprinting Alpha(Is Swim Sprinting Alpha:double=0.25) -> (Output_Get:double)
   Set Is Swim Sprinting Alpha(execute=@n13.else, Is Swim Sprinting Alpha=1)
   n14 = Branch(execute=@n3.then_6, Condition=Get bIsCrouching())
   > Set Is Crouching Alpha(Is Crouching Alpha:float=1) -> (Output_Get:float)
   Set Is Crouching Alpha(execute=@n14.else, Is Crouching Alpha=1)
   Set Is Crouching Float(execute=@n1.then_3, Is Crouching Float:float="To Float (Boolean)"(InBool=Get bIsCrouching())) -> (Output_Get:float)
   n15 = AND Boolean(A=Get bIsFalling(), B=NOT Boolean(A=Get bIsSkyDiving()))
   n16 = AND Boolean(A=Get bIsJumping(), B=NOT Boolean(A=Get bIsSkyDiving()))
   Set Is Jumping(execute=@n1.then_4, Is Jumping:float="To Float (Boolean)"(InBool=OR Boolean(A=@n16, B=@n15))) -> (Output_Get:float)
   Set SkyDiving_Float(execute=@n1.then_5, SkyDiving_Float:float="To Float (Boolean)"(InBool=Get bIsSkyDiving())) -> (Output_Get:float)
   n17 = Sequence!(execute=@n1.then_6) -> (then_0, then_1)
   > Is Valid(InputObject=GetOwnerSkeletalMeshComponent())
   > n18 = Cast To FortPlayerPawn(Object=GetOwningActor())
   > Is Valid(InputObject=@n18)
   n19 = GetAnimInstance(self=Get Mesh(self=@n18))
   > Is Valid(InputObject=@n19)
   > Set Anim Instance(Anim Instance:Anim Instance Object Reference=@n19) -> (Output_Get:Anim Instance Object Reference)
   n20 = Branch(execute=@n17.then_1, Condition=Get bIsInFrontEnd())
   Set CustomIdle is Playing Float(
    execute=@n20.else
    CustomIdle is Playing Float:float=GetCurveValue(self=Get Anim Instance() -> (Anim Instance:Anim Instance Object Reference), CurveName=Frontend_Idle_Bypass)
   ) -> (Output_Get:float)
   Set Is In Front End Alpha(execute=@n1.then_7, Is In Front End Alpha:double="To Float (Boolean)"(InBool=Get bIsInFrontEnd())) -> (Output_Get:double)
  ]}
]
variables [
 Anim Instance {type=AnimInstance, default=null}
 CustomIdle is Playing Float {type=number}
 Gravity Override Params {type=GravityOverrideParamsStruct}
 GravityOverride {type=Vector}
 Is Crouching Alpha {type=number}
 Is Crouching Float {type=number}
 Is In Front End Alpha {type=number}
 Is Jumping {type=number}
 Is Sky Diving Skirt Alpha {type=number}
 Is Sliding Alpha {type=number}
 Is Surface Swimming Alpha {type=number}
 Is Swim Sprinting Alpha {type=number}
 Is Tactical Sprinting Alpha {type=number}
 Jumping Skirt Alpha {type=number}
 SkirtAlpha {type=number}
 SkyDiving_Float {type=number}
]
