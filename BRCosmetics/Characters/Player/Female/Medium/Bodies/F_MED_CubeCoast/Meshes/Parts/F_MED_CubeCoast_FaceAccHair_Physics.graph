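# Physics asset for the CubeCoast female medium FaceAccHair mesh, defining bodies and constraints for dynamic hair, earrings and a left-side flower.
# Two braids, three back-hair strands and two earrings form capsule chains from head through zero-size MSTR bodies; the flower constrains directly to head.
# Chain links have limited swing and tight twist, while earring and back-hair roots swing freely.
# Head, neck, spine, clavicle and upper-arm capsules approximate the body.
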PhysicsAsset /BRCosmetics/Characters/Player/Female/Medium/Bodies/F_MED_CubeCoast/Meshes/Parts/F_MED_CubeCoast_FaceAccHair_Physics
bodies {
 clavicle_l [{shapeType=Capsule, length=12, radius=5.5, shapeName=clavicle_l_capsule, center=<3.5 -1 -1>, rotation=<rot 80 -0.0003 -0.0003>}]
 clavicle_r [{shapeType=Capsule, length=12, radius=5.5, shapeName=clavicle_r_capsule, center=<-3.5 1 1>, rotation=<rot 80 -0.0003 -0.0004>}]
 dyn_Earring_01_l [{shapeType=Capsule, length=1, radius=0.75, shapeName=dyn_Earring_01_l_capsule, center=<2.5 0 0>, rotation=<rot 90 0 0>}]
 dyn_Earring_01_r [{shapeType=Capsule, length=1, radius=0.75, shapeName=dyn_Earring_01_r_capsule, center=<2.5 0 0>, rotation=<rot 90 0 0>}]
 dyn_Earring_l_MSTR [{shapeType=Capsule, length=0, radius=0, shapeName=dyn_Earring_l_MSTR_capsule, rotation=<rot 90 0 0>}]
 dyn_Earring_r_MSTR [{shapeType=Capsule, length=0, radius=0, shapeName=dyn_Earring_r_MSTR_capsule, rotation=<rot 90 0 0>}]
 dyn_Flower_01_l [{shapeType=Capsule, length=0.5, radius=3.75, shapeName=dyn_Flower_01_l_capsule, center=<3.5 0 0>, rotation=<rot 90 0 0>}]
 dyn_HairBraid_01_l [{shapeType=Capsule, length=3, radius=1, shapeName=dyn_HairBraid_01_l_capsule, center=<2.75 0 0>, rotation=<rot 90 0 0>}]
 dyn_HairBraid_01_r [{shapeType=Capsule, length=3, radius=1, shapeName=dyn_HairBraid_01_r_capsule, center=<2.75 0 0>, rotation=<rot 90 0 0>}]
 dyn_HairBraid_02_l [{shapeType=Capsule, length=3, radius=1, shapeName=dyn_HairBraid_02_l_capsule, center=<2.75 0 0>, rotation=<rot 90 0 0>}]
 dyn_HairBraid_02_r [{shapeType=Capsule, length=3, radius=1, shapeName=dyn_HairBraid_02_r_capsule, center=<2.75 0 0>, rotation=<rot 90 0 0>}]
 dyn_HairBraid_03_l [{shapeType=Capsule, length=3, radius=1, shapeName=dyn_HairBraid_03_l_capsule, center=<2.75 0 0.25>, rotation=<rot 90 0 0>}]
 dyn_HairBraid_03_r [{shapeType=Capsule, length=3, radius=1, shapeName=dyn_HairBraid_03_r_capsule, center=<2.75 0 0.25>, rotation=<rot 90 0 0>}]
 dyn_HairBraid_04_l [{shapeType=Capsule, length=3, radius=1, shapeName=dyn_HairBraid_04_l_capsule, center=<2.75 0 0.25>, rotation=<rot 90 0 0>}]
 dyn_HairBraid_04_r [{shapeType=Capsule, length=3, radius=1, shapeName=dyn_HairBraid_04_r_capsule, center=<2.75 0 0.25>, rotation=<rot 90 0 0>}]
 dyn_HairBraid_05_l [{shapeType=Capsule, length=3, radius=1, shapeName=dyn_HairBraid_05_l_capsule, center=<2.75 0 0.15>, rotation=<rot 90 0 0>}]
 dyn_HairBraid_05_r [{shapeType=Capsule, length=3, radius=1, shapeName=dyn_HairBraid_05_r_capsule, center=<2.75 0 0.15>, rotation=<rot 90 0 0>}]
 dyn_HairBraid_l_MSTR [{shapeType=Capsule, length=0, radius=0, shapeName=dyn_HairBraid_l_MSTR_capsule, rotation=<rot 90 0 0>}]
 dyn_HairBraid_r_MSTR [{shapeType=Capsule, length=0, radius=0, shapeName=dyn_HairBraid_r_MSTR_capsule, rotation=<rot 90 0 0>}]
 dyn_Hair_bk_01_ct [{shapeType=Capsule, length=3, radius=2, shapeName=dyn_Hair_bk_01_ct_capsule, center=<3.75 0 0>, rotation=<rot 0 0 90>}]
 dyn_Hair_bk_01_l [{shapeType=Capsule, length=3, radius=2, shapeName=dyn_Hair_bk_01_l_capsule, center=<3.75 0 0>, rotation=<rot -0.0002 0 140>}]
 dyn_Hair_bk_01_r [{shapeType=Capsule, length=3, radius=2, shapeName=dyn_Hair_bk_01_r_capsule, center=<3.75 0 0>, rotation=<rot -0.0003 0 40>}]
 dyn_Hair_bk_02_ct [{shapeType=Capsule, length=5, radius=3, shapeName=dyn_Hair_bk_02_ct_capsule, center=<4 0 -2>, rotation=<rot 0 0 90>}]
 dyn_Hair_bk_02_l [
  {shapeType=Capsule, length=5, radius=3, shapeName=dyn_Hair_bk_02_l_capsule, center=<4 0 -1>, rotation=<rot -0.0002 0 130>}
  {shapeType=Capsule, length=4, radius=3, shapeName=dyn_Hair_bk_02_l_capsule, center=<6.8408 2.4466 -3.068>, rotation=<rot -0.0006 0 130>}
 ]
 dyn_Hair_bk_02_r [
  {shapeType=Capsule, length=5, radius=3, shapeName=dyn_Hair_bk_02_r_capsule, center=<4 0 -1>, rotation=<rot -0.0003 0 50>}
  {shapeType=Capsule, length=4, radius=3, shapeName=dyn_Hair_bk_02_r_capsule, center=<6.8408 -2.5087 -2.9475>, rotation=<rot -0.0002 0 50>}
 ]
 dyn_Hair_bk_03_ct [{shapeType=Capsule, length=10, radius=3.5, shapeName=dyn_Hair_bk_03_ct_capsule, center=<3.5 0 -2>, rotation=<rot 0 0 90>}]
 dyn_Hair_bk_03_l [
  {shapeType=Capsule, length=4, radius=5, shapeName=dyn_Hair_bk_03_l_capsule, center=<3.1644 3 -2.5>, rotation=<rot 0 0 90>}
  {shapeType=Capsule, length=8, radius=4, shapeName=dyn_Hair_bk_03_l_capsule, center=<8 3 -0.5>, rotation=<rot 0 0 90>}
 ]
 dyn_Hair_bk_03_r [
  {shapeType=Capsule, length=4, radius=5, shapeName=dyn_Hair_bk_03_r_capsule, center=<3.173 -3 -2.5>, rotation=<rot 0 0 90>}
  {shapeType=Capsule, length=8, radius=4, shapeName=dyn_Hair_bk_03_r_capsule, center=<8 -3 -1>, rotation=<rot 0 0 90>}
 ]
 dyn_Hair_bk_04_ct [{shapeType=Capsule, length=11, radius=4, shapeName=dyn_Hair_bk_04_ct_capsule, center=<5 0 0.5>, rotation=<rot 0 0 90>}]
 dyn_Hair_bk_04_l [{shapeType=Capsule, length=11, radius=3.5, shapeName=dyn_Hair_bk_04_l_capsule, center=<5 0 0.5>, rotation=<rot 0 0 90>}]
 dyn_Hair_bk_04_r [{shapeType=Capsule, length=11, radius=3.5, shapeName=dyn_Hair_bk_04_r_capsule, center=<5 0 0.5>, rotation=<rot 0 0 90>}]
 dyn_Hair_bk_05_ct [{shapeType=Capsule, length=10, radius=5, shapeName=dyn_Hair_bk_05_ct_capsule, center=<5 0 1.5>, rotation=<rot 0 0 90>}]
 dyn_Hair_bk_05_l [{shapeType=Capsule, length=10, radius=5, shapeName=dyn_Hair_bk_05_l_capsule, center=<5 0 2>, rotation=<rot 0 0 90>}]
 dyn_Hair_bk_05_r [{shapeType=Capsule, length=10, radius=5, shapeName=dyn_Hair_bk_05_r_capsule, center=<5 0 2>, rotation=<rot 0 0 90>}]
 dyn_Hair_bk_MSTR [{shapeType=Capsule, length=0, radius=0, shapeName=dyn_Hair_bk_MSTR_capsule, rotation=<rot 90 0 0>}]
 dyn_Hair_bk_l_MSTR [{shapeType=Capsule, length=0, radius=0, shapeName=dyn_Hair_bk_l_MSTR_capsule, rotation=<rot 90 0 0>}]
 dyn_Hair_bk_r_MSTR [{shapeType=Capsule, length=0, radius=0, shapeName=dyn_Hair_bk_r_MSTR_capsule, rotation=<rot 90 0 0>}]
 head [{shapeType=Capsule, length=6, radius=6, shapeName=head_capsule, center=<3.5 3 0>, rotation=<rot 90 0 0>}]
 neck_01 [{shapeType=Capsule, length=8, radius=4.5, shapeName=neck_01_capsule, center=<3.6683 -0.2876 0>, rotation=<rot 90 0 0>}]
 spine_03 [{shapeType=Capsule, length=15, radius=10, shapeName=spine_04_capsule, center=<0.229 2 0>, rotation=<rot 90 0 10.0001>}]
 spine_05 [
  {shapeType=Capsule, length=12, radius=9.25, shapeName=spine_05_capsule, center=<-2.5963 0.209 0>}
  {shapeType=Capsule, length=12, radius=7.75, shapeName=spine_05_capsule, center=<-6.4845 2.8778 0>}
  {shapeType=Capsule, length=3, radius=1.85, shapeName=spine_05_capsule, center=<3.2392 6.9657 0>, rotation=<rot 90 7.125 37.1254>}
 ]
 upperarm_l [{shapeType=Capsule, length=20, radius=3.5, shapeName=upperarm_l_capsule, center=<12 -0.5 0>, rotation=<rot 90 0 0>}]
 upperarm_r [{shapeType=Capsule, length=20, radius=3.5, shapeName=upperarm_r_capsule, center=<-12 0.5 0>, rotation=<rot 90 0 0>}]
}
constraints [
 {
  bone1Name=dyn_Earring_l_MSTR
  bone2Name=head
  swing1LimitDegrees=45
  swing1Motion=Free
  swing2LimitDegrees=45
  swing2Motion=Free
  twistLimitDegrees=45
  twistMotion=Free
 }
 {bone1Name=dyn_Earring_01_l, bone2Name=dyn_Earring_l_MSTR, swing1LimitDegrees=35, swing2LimitDegrees=35, twistLimitDegrees=2}
 {
  bone1Name=dyn_Earring_r_MSTR
  bone2Name=head
  swing1LimitDegrees=45
  swing1Motion=Free
  swing2LimitDegrees=45
  swing2Motion=Free
  twistLimitDegrees=45
  twistMotion=Free
 }
 {bone1Name=dyn_Earring_01_r, bone2Name=dyn_Earring_r_MSTR, swing1LimitDegrees=35, swing2LimitDegrees=35, twistLimitDegrees=2}
 {bone1Name=dyn_HairBraid_l_MSTR, bone2Name=head, swing1LimitDegrees=35, swing2LimitDegrees=35, twistLimitDegrees=2}
 {bone1Name=dyn_HairBraid_01_l, bone2Name=dyn_HairBraid_l_MSTR, swing1LimitDegrees=8, swing2LimitDegrees=35, twistLimitDegrees=2}
 {bone1Name=dyn_HairBraid_02_l, bone2Name=dyn_HairBraid_01_l, swing1LimitDegrees=8, swing2LimitDegrees=35, twistLimitDegrees=2}
 {bone1Name=dyn_HairBraid_03_l, bone2Name=dyn_HairBraid_02_l, swing1LimitDegrees=8, swing2LimitDegrees=35, twistLimitDegrees=2}
 {bone1Name=dyn_HairBraid_04_l, bone2Name=dyn_HairBraid_03_l, swing1LimitDegrees=8, swing2LimitDegrees=35, twistLimitDegrees=2}
 {bone1Name=dyn_HairBraid_05_l, bone2Name=dyn_HairBraid_04_l, swing1LimitDegrees=8, swing2LimitDegrees=35, twistLimitDegrees=2}
 {bone1Name=dyn_HairBraid_r_MSTR, bone2Name=head, swing1LimitDegrees=35, swing2LimitDegrees=35, twistLimitDegrees=2}
 {bone1Name=dyn_HairBraid_01_r, bone2Name=dyn_HairBraid_r_MSTR, swing1LimitDegrees=15, swing2LimitDegrees=35, twistLimitDegrees=2}
 {bone1Name=dyn_HairBraid_02_r, bone2Name=dyn_HairBraid_01_r, swing1LimitDegrees=15, swing2LimitDegrees=35, twistLimitDegrees=2}
 {bone1Name=dyn_HairBraid_03_r, bone2Name=dyn_HairBraid_02_r, swing1LimitDegrees=15, swing2LimitDegrees=35, twistLimitDegrees=2}
 {bone1Name=dyn_HairBraid_04_r, bone2Name=dyn_HairBraid_03_r, swing1LimitDegrees=15, swing2LimitDegrees=35, twistLimitDegrees=2}
 {bone1Name=dyn_HairBraid_05_r, bone2Name=dyn_HairBraid_04_r, swing1LimitDegrees=15, swing2LimitDegrees=35, twistLimitDegrees=2}
 {
  bone1Name=dyn_Hair_bk_MSTR
  bone2Name=head
  swing1LimitDegrees=45
  swing1Motion=Free
  swing2LimitDegrees=45
  swing2Motion=Free
  twistLimitDegrees=45
  twistMotion=Free
 }
 {bone1Name=dyn_Hair_bk_01_ct, bone2Name=dyn_Hair_bk_MSTR, swing1LimitDegrees=15, swing2LimitDegrees=45, twistLimitDegrees=2}
 {bone1Name=dyn_Hair_bk_02_ct, bone2Name=dyn_Hair_bk_01_ct, swing1LimitDegrees=15, swing2LimitDegrees=45, twistLimitDegrees=2}
 {bone1Name=dyn_Hair_bk_03_ct, bone2Name=dyn_Hair_bk_02_ct, swing1LimitDegrees=15, swing2LimitDegrees=45, twistLimitDegrees=2}
 {bone1Name=dyn_Hair_bk_04_ct, bone2Name=dyn_Hair_bk_03_ct, swing1LimitDegrees=15, swing2LimitDegrees=45, twistLimitDegrees=2}
 {bone1Name=dyn_Hair_bk_05_ct, bone2Name=dyn_Hair_bk_04_ct, swing1LimitDegrees=15, swing2LimitDegrees=45, twistLimitDegrees=2}
 {
  bone1Name=dyn_Hair_bk_l_MSTR
  bone2Name=head
  swing1LimitDegrees=45
  swing1Motion=Free
  swing2LimitDegrees=45
  swing2Motion=Free
  twistLimitDegrees=45
  twistMotion=Free
 }
 {bone1Name=dyn_Hair_bk_01_l, bone2Name=dyn_Hair_bk_l_MSTR, swing1LimitDegrees=15, swing2LimitDegrees=45, twistLimitDegrees=2}
 {bone1Name=dyn_Hair_bk_02_l, bone2Name=dyn_Hair_bk_01_l, swing1LimitDegrees=15, swing2LimitDegrees=45, twistLimitDegrees=2}
 {bone1Name=dyn_Hair_bk_03_l, bone2Name=dyn_Hair_bk_02_l, swing1LimitDegrees=15, swing2LimitDegrees=45, twistLimitDegrees=2}
 {bone1Name=dyn_Hair_bk_04_l, bone2Name=dyn_Hair_bk_03_l, swing1LimitDegrees=15, swing2LimitDegrees=45, twistLimitDegrees=2}
 {bone1Name=dyn_Hair_bk_05_l, bone2Name=dyn_Hair_bk_04_l, swing1LimitDegrees=15, swing2LimitDegrees=45, twistLimitDegrees=2}
 {
  bone1Name=dyn_Hair_bk_r_MSTR
  bone2Name=head
  swing1LimitDegrees=45
  swing1Motion=Free
  swing2LimitDegrees=45
  swing2Motion=Free
  twistLimitDegrees=45
  twistMotion=Free
 }
 {bone1Name=dyn_Hair_bk_01_r, bone2Name=dyn_Hair_bk_r_MSTR, swing1LimitDegrees=15, swing2LimitDegrees=45, twistLimitDegrees=2}
 {bone1Name=dyn_Hair_bk_02_r, bone2Name=dyn_Hair_bk_01_r, swing1LimitDegrees=15, swing2LimitDegrees=45, twistLimitDegrees=2}
 {bone1Name=dyn_Hair_bk_03_r, bone2Name=dyn_Hair_bk_02_r, swing1LimitDegrees=15, swing2LimitDegrees=45, twistLimitDegrees=2}
 {bone1Name=dyn_Hair_bk_04_r, bone2Name=dyn_Hair_bk_03_r, swing1LimitDegrees=15, swing2LimitDegrees=45, twistLimitDegrees=2}
 {bone1Name=dyn_Hair_bk_05_r, bone2Name=dyn_Hair_bk_04_r, swing1LimitDegrees=15, swing2LimitDegrees=45, twistLimitDegrees=2}
 {
  bone1Name=head
  bone2Name=spine_05
  swing1LimitDegrees=45
  swing1Motion=Free
  swing2LimitDegrees=45
  swing2Motion=Free
  twistLimitDegrees=45
  twistMotion=Free
 }
 {bone1Name=dyn_Flower_01_l, bone2Name=head, swing1LimitDegrees=25, swing2LimitDegrees=25, twistLimitDegrees=2}
]
