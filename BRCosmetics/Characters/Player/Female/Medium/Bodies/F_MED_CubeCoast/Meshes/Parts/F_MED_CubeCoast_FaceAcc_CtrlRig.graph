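# Control Rig for the F_MED_CubeCoast face accessory part that poses the back hair and braid bones to follow body motion.
# Rotation constraints first blend five hair master bones equally between spine_05 and head.
# Four head spherical pose readers then weight corrective rotations on those bones.
# Two right upper-arm readers, multiplied by the Is Pistol Alpha and Is Rifle Alpha variables, add further back-hair offsets for weapon poses.
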ControlRigBlueprint /BRCosmetics/Characters/Player/Female/Medium/Bodies/F_MED_CubeCoast/Meshes/Parts/F_MED_CubeCoast_FaceAcc_CtrlRig
graphs [
 RigVMFunctionLibrary {links [], nodes []}
 RigVMModel {nodes [
   RigUnit_BeginExecution = &RigUnit_BeginExecution()
   &RigUnit_RotationConstraintLocalSpaceOffset(
    AdvancedSettings=(RotationOrderForFilter=XZY)
    Child=(Type=Bone,Name="dyn_Hair_bk_l_MSTR")
    Filter=(bX=True,bY=True,bZ=True)
    Parents=((Item=(Type=Bone,Name="spine_05"),Weight=0.5),(Item=(Type=Bone,Name="head"),Weight=0.5))
    Weight=1
    ExecutePin=@RigUnit_BeginExecution.ExecutePin
   )
   > &RigUnit_RotationConstraintLocalSpaceOffset(
    AdvancedSettings=(RotationOrderForFilter=XZY)
    Child=(Type=Bone,Name="dyn_Hair_bk_MSTR")
    Filter=(bX=True,bY=True,bZ=True)
    Parents=((Item=(Type=Bone,Name="spine_05"),Weight=0.5),(Item=(Type=Bone,Name="head"),Weight=0.5))
    Weight=1
   )
   > &RigUnit_RotationConstraintLocalSpaceOffset(
    AdvancedSettings=(RotationOrderForFilter=XZY)
    Child=(Type=Bone,Name="dyn_Hair_bk_r_MSTR")
    Filter=(bX=True,bY=True,bZ=True)
    Parents=((Item=(Type=Bone,Name="spine_05"),Weight=0.5),(Item=(Type=Bone,Name="head"),Weight=0.5))
    Weight=1
    bMaintainOffset=true
   )
   > &RigUnit_RotationConstraintLocalSpaceOffset(
    AdvancedSettings=(RotationOrderForFilter=XZY)
    Child=(Type=Bone,Name="dyn_HairBraid_l_MSTR")
    Filter=(bX=True,bY=True,bZ=True)
    Parents=((Item=(Type=Bone,Name="spine_05"),Weight=0.5),(Item=(Type=Bone,Name="head"),Weight=0.5))
    Weight=1
    bMaintainOffset=true
   )
   > &RigUnit_RotationConstraintLocalSpaceOffset(
    AdvancedSettings=(RotationOrderForFilter=XZY)
    Child=(Type=Bone,Name="dyn_HairBraid_r_MSTR")
    Filter=(bX=True,bY=True,bZ=True)
    Parents=((Item=(Type=Bone,Name="spine_05"),Weight=0.5),(Item=(Type=Bone,Name="head"),Weight=0.5))
    Weight=1
    bMaintainOffset=true
   )
   > n1 = &RigUnit_SphericalPoseReader(
    ActiveRegionScaleFactors=(PositiveWidth=1,NegativeWidth=1,PositiveHeight=1,NegativeHeight=1)
    ActiveRegionSize=0.1
    Debug=(bDrawDebug=false,DebugScale=25,DebugSegments=20,DebugThickness=0.25)
    DriverAxis=(X=0,Y=-1,Z=0)
    DriverItem=(Type=Bone,Name="head")
    FalloffRegionScaleFactors=(PositiveWidth=2,NegativeWidth=2,PositiveHeight=2,NegativeHeight=2)
    FalloffSize=0.2
    FlipWidthScaling=false
    OptionalParentItem=(Type=Bone,Name="neck_01")
    RotationOffset=(X=-90,Y=0,Z=-55)
   )
   > &RigUnit_ModifyTransforms(
    ItemToModify=((Item=(Type=Bone,Name="dyn_HairBraid_l_MSTR"),Transform=(Rotation=(X=0,Y=0,Z=0,W=1),Translation=(X=0,Y=0,Z=0),Scale3D=(X=1,Y=1,Z=1))))
    ItemToModify.0.Transform.Rotation=&RigVMFunction_MathQuaternionFromEuler(Euler=(X=0,Y=40,Z=0)).Result
    Weight=@n1.OutputParam
   )
   > &RigUnit_ModifyTransforms(
    ItemToModify=((Item=(Type=Bone,Name="dyn_HairBraid_r_MSTR"),Transform=(Rotation=(X=0,Y=0,Z=0,W=1),Translation=(X=0,Y=0,Z=0),Scale3D=(X=1,Y=1,Z=1))))
    ItemToModify.0.Transform.Rotation=&RigVMFunction_MathQuaternionFromEuler(Euler=(X=0,Y=40,Z=0)).Result
    Weight=@n1.OutputParam
   )
   > &RigUnit_ModifyTransforms(
    ItemToModify=((Item=(Type=Bone,Name="dyn_Hair_bk_r_MSTR"),Transform=(Rotation=(X=0,Y=0,Z=0,W=1),Translation=(X=0,Y=0,Z=0),Scale3D=(X=1,Y=1,Z=1))))
    ItemToModify.0.Transform.Rotation=&RigVMFunction_MathQuaternionFromEuler(Euler=(X=0,Y=-20,Z=0)).Result
    Weight=@n1.OutputParam
   )
   > &RigUnit_ModifyTransforms(
    ItemToModify=((Item=(Type=Bone,Name="dyn_Hair_bk_l_MSTR"),Transform=(Rotation=(X=0,Y=0,Z=0,W=1),Translation=(X=0,Y=0,Z=0),Scale3D=(X=1,Y=1,Z=1))))
    ItemToModify.0.Transform.Rotation=&RigVMFunction_MathQuaternionFromEuler(Euler=(X=0,Y=-20,Z=0)).Result
    Weight=@n1.OutputParam
   )
   > &RigUnit_ModifyTransforms(
    ItemToModify=((Item=(Type=Bone,Name="dyn_Hair_bk_MSTR"),Transform=(Rotation=(X=0,Y=0,Z=0,W=1),Translation=(X=0,Y=0,Z=0),Scale3D=(X=1,Y=1,Z=1))))
    ItemToModify.0.Transform.Rotation=&RigVMFunction_MathQuaternionFromEuler(Euler=(X=0,Y=-20,Z=0)).Result
    Weight=@n1.OutputParam
   )
   > &RigUnit_ModifyTransforms(
    ItemToModify=((Item=(Type=Bone,Name="dyn_Hair_bk_02_r"),Transform=(Rotation=(X=0,Y=0,Z=0,W=1),Translation=(X=0,Y=0,Z=0),Scale3D=(X=1,Y=1,Z=1))))
    ItemToModify.0.Transform.Rotation=&RigVMFunction_MathQuaternionFromEuler(Euler=(X=0,Y=0,Z=0)).Result
    Weight=@n1.OutputParam
   )
   > &RigUnit_ModifyTransforms(
    ItemToModify=((Item=(Type=Bone,Name="dyn_Hair_bk_02_l"),Transform=(Rotation=(X=0,Y=0,Z=0,W=1),Translation=(X=0,Y=0,Z=0),Scale3D=(X=1,Y=1,Z=1))))
    ItemToModify.0.Transform.Rotation=&RigVMFunction_MathQuaternionFromEuler(Euler=(X=0,Y=0,Z=0)).Result
    Weight=@n1.OutputParam
   )
   > &RigUnit_ModifyTransforms(
    ItemToModify=((Item=(Type=Bone,Name="dyn_Hair_bk_02_ct"),Transform=(Rotation=(X=0,Y=0,Z=0,W=1),Translation=(X=0,Y=0,Z=0),Scale3D=(X=1,Y=1,Z=1))))
    ItemToModify.0.Transform.Rotation=&RigVMFunction_MathQuaternionFromEuler(Euler=(X=0,Y=0,Z=0)).Result
    Weight=@n1.OutputParam
   )
   > n2 = &RigUnit_SphericalPoseReader(
    ActiveRegionScaleFactors=(PositiveWidth=0.25,NegativeWidth=0.25,PositiveHeight=1,NegativeHeight=1)
    ActiveRegionSize=0.1
    Debug=(bDrawDebug=false,DebugScale=25,DebugSegments=20,DebugThickness=0.25)
    DriverAxis=(X=0,Y=1,Z=0)
    DriverItem=(Type=Bone,Name="head")
    FalloffRegionScaleFactors=(PositiveWidth=0.75,NegativeWidth=0.5,PositiveHeight=4,NegativeHeight=4)
    FalloffSize=0.2
    OptionalParentItem=(Type=Bone,Name="spine_05")
    RotationOffset=(X=90,Y=0,Z=25)
   )
   > &RigUnit_ModifyTransforms(
    ItemToModify=((Item=(Type=Bone,Name="dyn_HairBraid_l_MSTR"),Transform=(Rotation=(X=0,Y=0,Z=0,W=1),Translation=(X=0,Y=0,Z=0),Scale3D=(X=1,Y=1,Z=1))))
    ItemToModify.0.Transform.Rotation=&RigVMFunction_MathQuaternionFromEuler(Euler=(X=0,Y=-40,Z=0)).Result
    Weight=@n2.OutputParam
   )
   > &RigUnit_ModifyTransforms(
    ItemToModify=((Item=(Type=Bone,Name="dyn_HairBraid_r_MSTR"),Transform=(Rotation=(X=0,Y=0,Z=0,W=1),Translation=(X=0,Y=0,Z=0),Scale3D=(X=1,Y=1,Z=1))))
    ItemToModify.0.Transform.Rotation=&RigVMFunction_MathQuaternionFromEuler(Euler=(X=0,Y=-40,Z=0)).Result
    Weight=@n2.OutputParam
   )
   > &RigUnit_ModifyTransforms(
    ItemToModify=((Item=(Type=Bone,Name="dyn_Hair_bk_r_MSTR"),Transform=(Rotation=(X=0,Y=0,Z=0,W=1),Translation=(X=0,Y=0,Z=0),Scale3D=(X=1,Y=1,Z=1))))
    ItemToModify.0.Transform.Rotation=&RigVMFunction_MathQuaternionFromEuler(Euler=(X=0,Y=-10,Z=0)).Result
    Weight=@n2.OutputParam
   )
   > &RigUnit_ModifyTransforms(
    ItemToModify=((Item=(Type=Bone,Name="dyn_Hair_bk_l_MSTR"),Transform=(Rotation=(X=0,Y=0,Z=0,W=1),Translation=(X=0,Y=0,Z=0),Scale3D=(X=1,Y=1,Z=1))))
    ItemToModify.0.Transform.Rotation=&RigVMFunction_MathQuaternionFromEuler(Euler=(X=0,Y=-10,Z=0)).Result
    Weight=@n2.OutputParam
   )
   > &RigUnit_ModifyTransforms(
    ItemToModify=((Item=(Type=Bone,Name="dyn_Hair_bk_MSTR"),Transform=(Rotation=(X=0,Y=0,Z=0,W=1),Translation=(X=0,Y=0,Z=0),Scale3D=(X=1,Y=1,Z=1))))
    ItemToModify.0.Transform.Rotation=&RigVMFunction_MathQuaternionFromEuler(Euler=(X=0,Y=-10,Z=0)).Result
    Weight=@n2.OutputParam
   )
   > n3 = &RigUnit_SphericalPoseReader(
    ActiveRegionScaleFactors=(PositiveWidth=1,NegativeWidth=0.5,PositiveHeight=1,NegativeHeight=1)
    ActiveRegionSize=0.1
    Debug=(bDrawDebug=false,DebugScale=25,DebugSegments=20,DebugThickness=0.25)
    DriverAxis=(X=0,Y=-1,Z=0)
    DriverItem=(Type=Bone,Name="head")
    FalloffRegionScaleFactors=(PositiveWidth=1,NegativeWidth=0.5,PositiveHeight=1,NegativeHeight=1)
    FalloffSize=0.2
    OptionalParentItem=(Type=Bone,Name="neck_02")
    RotationOffset=(X=-55,Y=0,Z=-10)
   )
   > &RigUnit_ModifyTransforms(
    ItemToModify=((Item=(Type=Bone,Name="dyn_Hair_bk_r_MSTR"),Transform=(Rotation=(X=0,Y=0,Z=0,W=1),Translation=(X=0,Y=0,Z=0),Scale3D=(X=1,Y=1,Z=1))))
    ItemToModify.0.Transform.Rotation=&RigVMFunction_MathQuaternionFromEuler(Euler=(X=0,Y=-15,Z=0)).Result
    Weight=@n3.OutputParam
   )
   > &RigUnit_ModifyTransforms(
    ItemToModify=((Item=(Type=Bone,Name="dyn_HairBraid_l_MSTR"),Transform=(Rotation=(X=0,Y=0,Z=0,W=1),Translation=(X=0,Y=0,Z=0),Scale3D=(X=1,Y=1,Z=1))))
    ItemToModify.0.Transform.Rotation=&RigVMFunction_MathQuaternionFromEuler(Euler=(X=0,Y=-15,Z=0)).Result
    Weight=@n3.OutputParam
   )
   > n4 = &RigUnit_SphericalPoseReader(
    ActiveRegionScaleFactors=(PositiveWidth=1,NegativeWidth=0.5,PositiveHeight=1,NegativeHeight=1)
    ActiveRegionSize=0.1
    Debug=(bDrawDebug=false,DebugScale=25,DebugSegments=20,DebugThickness=0.25)
    DriverAxis=(X=0,Y=-1,Z=0)
    DriverItem=(Type=Bone,Name="head")
    FalloffRegionScaleFactors=(PositiveWidth=1,NegativeWidth=0.5,PositiveHeight=1,NegativeHeight=1)
    FalloffSize=0.2
    OptionalParentItem=(Type=Bone,Name="neck_02")
    RotationOffset=(X=235,Y=0,Z=-10)
   )
   > &RigUnit_ModifyTransforms(
    ItemToModify=((Item=(Type=Bone,Name="dyn_Hair_bk_l_MSTR"),Transform=(Rotation=(X=0,Y=0,Z=0,W=1),Translation=(X=0,Y=0,Z=0),Scale3D=(X=1,Y=1,Z=1))))
    ItemToModify.0.Transform.Rotation=&RigVMFunction_MathQuaternionFromEuler(Euler=(X=0,Y=-15,Z=0)).Result
    Weight=@n4.OutputParam
   )
   > &RigUnit_ModifyTransforms(
    ItemToModify=((Item=(Type=Bone,Name="dyn_HairBraid_r_MSTR"),Transform=(Rotation=(X=0,Y=0,Z=0,W=1),Translation=(X=0,Y=0,Z=0),Scale3D=(X=1,Y=1,Z=1))))
    ItemToModify.0.Transform.Rotation=&RigVMFunction_MathQuaternionFromEuler(Euler=(X=0,Y=-15,Z=0)).Result
    Weight=@n4.OutputParam
   )
   > n5 = &RigUnit_SphericalPoseReader(
    ActiveRegionScaleFactors=(PositiveWidth=0.5,NegativeWidth=0.5,PositiveHeight=1,NegativeHeight=0.5)
    ActiveRegionSize=0.1
    Debug=(bDrawDebug=false,bDraw2D=false,bDrawLocalAxes=false,DebugScale=25,DebugSegments=20,DebugThickness=0.25)
    DriverAxis=(X=-1,Y=0,Z=0)
    DriverItem=(Type=Bone,Name="upperarm_r")
    FalloffRegionScaleFactors=(PositiveWidth=0.5,NegativeWidth=0.5,PositiveHeight=1,NegativeHeight=0.5)
    FalloffSize=0.45
    OptionalParentItem=(Type=Bone,Name="spine_05")
    RotationOffset=(X=-90,Y=40,Z=45)
   )
   Multiply = &RigVMFunction_MathDoubleMul(A=@n5.OutputParam, B=RigVMVariableNode() {variable=Is Pistol Alpha}.Value)
   > &RigUnit_ModifyTransforms(
    ItemToModify=((Item=(Type=Bone,Name="dyn_Hair_bk_r_MSTR"),Transform=(Rotation=(X=0,Y=0,Z=0,W=1),Translation=(X=0,Y=0,Z=0),Scale3D=(X=1,Y=1,Z=1))))
    ItemToModify.0.Transform.Rotation=&RigVMFunction_MathQuaternionFromEuler(Euler=(X=10,Y=-20,Z=0)).Result
    Weight=@Multiply.Result
   )
   > &RigUnit_ModifyTransforms(
    ItemToModify=((Item=(Type=Bone,Name="dyn_Hair_bk_l_MSTR"),Transform=(Rotation=(X=0,Y=0,Z=0,W=1),Translation=(X=0,Y=0,Z=0),Scale3D=(X=1,Y=1,Z=1))))
    ItemToModify.0.Transform.Rotation=&RigVMFunction_MathQuaternionFromEuler(Euler=(X=-10,Y=-20,Z=0)).Result
    Weight=@Multiply.Result
   )
   > &RigUnit_ModifyTransforms(
    ItemToModify=((Item=(Type=Bone,Name="dyn_Hair_bk_MSTR"),Transform=(Rotation=(X=0,Y=0,Z=0,W=1),Translation=(X=0,Y=0,Z=0),Scale3D=(X=1,Y=1,Z=1))))
    ItemToModify.0.Transform.Rotation=&RigVMFunction_MathQuaternionFromEuler(Euler=(X=0,Y=-20,Z=0)).Result
    Weight=@Multiply.Result
   )
   > n6 = &RigUnit_SphericalPoseReader(
    ActiveRegionScaleFactors=(PositiveWidth=0.35,NegativeWidth=0.35,PositiveHeight=0.35,NegativeHeight=0.35)
    ActiveRegionSize=0.1
    Debug=(bDrawDebug=false,bDraw2D=false,bDrawLocalAxes=false,DebugScale=25,DebugSegments=20,DebugThickness=0.25)
    DriverAxis=(X=-1,Y=0,Z=0)
    DriverItem=(Type=Bone,Name="upperarm_r")
    FalloffRegionScaleFactors=(PositiveWidth=0.5,NegativeWidth=0.5,PositiveHeight=0.5,NegativeHeight=0.5)
    FalloffSize=0.2
    OptionalParentItem=(Type=Bone,Name="spine_05")
    RotationOffset=(X=-45,Y=0,Z=-30)
   )
   n7 = &RigVMFunction_MathDoubleMul(A=@n6.OutputParam, B=RigVMVariableNode() {variable=Is Rifle Alpha}.Value)
   > &RigUnit_ModifyTransforms(
    ItemToModify=((Item=(Type=Bone,Name="dyn_Hair_bk_r_MSTR"),Transform=(Rotation=(X=0,Y=0,Z=0,W=1),Translation=(X=0,Y=0,Z=0),Scale3D=(X=1,Y=1,Z=1))))
    ItemToModify.0.Transform.Rotation=&RigVMFunction_MathQuaternionFromEuler(Euler=(X=10,Y=-20,Z=0)).Result
    Weight=@n7.Result
   )
   > &RigUnit_ModifyTransforms(
    ItemToModify=((Item=(Type=Bone,Name="dyn_Hair_bk_l_MSTR"),Transform=(Rotation=(X=0,Y=0,Z=0,W=1),Translation=(X=0,Y=0,Z=0),Scale3D=(X=1,Y=1,Z=1))))
    ItemToModify.0.Transform.Rotation=&RigVMFunction_MathQuaternionFromEuler(Euler=(X=-10,Y=-20,Z=0)).Result
    Weight=@n7.Result
   )
   > &RigUnit_ModifyTransforms(
    ItemToModify=((Item=(Type=Bone,Name="dyn_Hair_bk_MSTR"),Transform=(Rotation=(X=0,Y=0,Z=0,W=1),Translation=(X=0,Y=0,Z=0),Scale3D=(X=1,Y=1,Z=1))))
    ItemToModify.0.Transform.Rotation=&RigVMFunction_MathQuaternionFromEuler(Euler=(X=0,Y=-20,Z=0)).Result
    Weight=@n7.Result
   )
  ]}
]
hierarchy [
 root {loc=<0 0 0>, rot=<0 0 0 1>}
 pelvis {parent=root, loc=<0 -0.5301 98.0571>, rot=<0.0127 0.707 -0.0127 -0.707>}
 spine_01 {parent=pelvis, loc=<6.1208 -0.9328 0>, rot=<0 0 0.034 -0.9994>}
 spine_02 {parent=spine_01, loc=<6.5996 -0.2095 0>, rot=<0 0 0.0238 -0.9997>}
 spine_03 {parent=spine_02, loc=<6.4914 -0.1237 0>, rot=<0 0 0.0149 -0.9999>}
 spine_04 {parent=spine_03, loc=<6.9598 0.0794 0>, rot=<0 0 -0.0132 -0.9999>}
 spine_05 {parent=spine_04, loc=<6.9987 0 0>, rot=<0 0 -0.0015 -1>}
 neck_01 {parent=spine_05, loc=<11.3138 0.1529 0>, rot=<0 0 -0.2003 -0.9797>}
 neck_02 {parent=neck_01, loc=<4.5627 0 0>, rot=<0 0 0.0504 -0.9987>}
 head {parent=neck_02, loc=<5.3914 0 0>, rot=<0 0 0.115 -0.9934>}
 dyn_HairBraid_l_MSTR {parent=head, loc=<1.9596 -1.8872 -6.5114>, rot=<0.0469 0.7055 0.7055 0.0469>}
 dyn_HairBraid_01_l {parent=dyn_HairBraid_l_MSTR, loc=<0 0 0>, rot=<0 0 0 1>}
 dyn_HairBraid_02_l {parent=dyn_HairBraid_01_l, loc=<6.4057 0 0>, rot=<0 0.0886 0 -0.9961>}
 dyn_HairBraid_03_l {parent=dyn_HairBraid_02_l, loc=<5.4324 0 0>, rot=<0 0.0835 0 -0.9965>}
 dyn_HairBraid_04_l {parent=dyn_HairBraid_03_l, loc=<5.3283 0 0>, rot=<0 0.0864 0 -0.9963>}
 dyn_HairBraid_05_l {parent=dyn_HairBraid_04_l, loc=<4.743 0 0>, rot=<0 0.0521 0 -0.9986>}
 dyn_HairBraid_06_l {parent=dyn_HairBraid_05_l, loc=<5.5993 0 0>, rot=<0 0 0 1>}
 dyn_Bangs_ft_MSTR {parent=head, loc=<12.754 4.7462 -1.5163>, rot=<0.4979 0.5021 0.5021 0.4979>}
 dyn_Bangs_ft_01_ct {parent=dyn_Bangs_ft_MSTR, loc=<0 0 0>, rot=<0 -0.1734 0 -0.9848>}
 dyn_Bangs_ft_02_l {parent=dyn_Bangs_ft_01_ct, loc=<2.2216 -1.087 -0.3185>, rot=<-0.1262 -0.1772 0.2141 -0.9523>}
 dyn_Bangs_ft_03_l {parent=dyn_Bangs_ft_02_l, loc=<2.4836 0 -0.2232>, rot=<0 -0.1829 0 -0.9831>}
 dyn_Bangs_ft_04_l {parent=dyn_Bangs_ft_03_l, loc=<2.4836 0 -0.2232>, rot=<0 -0.1829 0 -0.9831>}
 dyn_Bangs_ft_05_l {parent=dyn_Bangs_ft_04_l, loc=<2.4836 0 -0.2232>, rot=<0 -0.1829 0 -0.9831>}
 dyn_Bangs_ft_02_r {parent=dyn_Bangs_ft_01_ct, loc=<2.3211 0.8667 -0.2823>, rot=<0.0994 -0.1794 -0.1686 -0.9641>}
 dyn_Bangs_ft_03_r {parent=dyn_Bangs_ft_02_r, loc=<2.4836 0 -0.2232>, rot=<0 -0.1829 0 -0.9831>}
 dyn_Bangs_ft_04_r {parent=dyn_Bangs_ft_03_r, loc=<2.4836 0 -0.2232>, rot=<0 -0.1829 0 -0.9831>}
 dyn_Bangs_ft_05_r {parent=dyn_Bangs_ft_04_r, loc=<2.4836 0 -0.2232>, rot=<0 -0.1829 0 -0.9831>}
 dyn_Hair_bk_MSTR {parent=head, loc=<9.8456 -6.1269 0>, rot=<0.5 -0.5 0.5 -0.5>}
 dyn_Hair_bk_01_ct {parent=dyn_Hair_bk_MSTR, loc=<0 0 0>, rot=<0 -0.5708 0 -0.8211>}
 dyn_Hair_bk_02_ct {parent=dyn_Hair_bk_01_ct, loc=<7.522 0 0>, rot=<0 -0.0608 0 -0.9981>}
 dyn_Hair_bk_03_ct {parent=dyn_Hair_bk_02_ct, loc=<7.522 0 0>, rot=<0 0.077 0 -0.997>}
 dyn_Hair_bk_04_ct {parent=dyn_Hair_bk_03_ct, loc=<7.522 0 0>, rot=<0 -0.1598 0 -0.9871>}
 dyn_Hair_bk_05_ct {parent=dyn_Hair_bk_04_ct, loc=<7.522 0 0>, rot=<0 -0.0595 0 -0.9982>}
 dyn_Hair_bk_06_ct {parent=dyn_Hair_bk_05_ct, loc=<7.522 0 0>, rot=<0 0 0 1>}
 dyn_Hair_bk_l_MSTR {parent=head, loc=<9.6609 -1.634 -6.4062>, rot=<0.5 -0.5 0.5 -0.5>}
 dyn_Hair_bk_01_l {parent=dyn_Hair_bk_l_MSTR, loc=<0 0 0>, rot=<0 -0.4882 0 -0.8727>}
 dyn_Hair_bk_02_l {parent=dyn_Hair_bk_01_l, loc=<7.522 0 0>, rot=<0 -0.16 0 -0.9871>}
 dyn_Hair_bk_03_l {parent=dyn_Hair_bk_02_l, loc=<7.522 0 0>, rot=<0 0.1298 0 -0.9915>}
 dyn_Hair_bk_04_l {parent=dyn_Hair_bk_03_l, loc=<7.522 0 0>, rot=<0 -0.1239 0 -0.9923>}
 dyn_Hair_bk_05_l {parent=dyn_Hair_bk_04_l, loc=<7.522 0 0>, rot=<0 -0.1201 0 -0.9928>}
 dyn_Hair_bk_06_l {parent=dyn_Hair_bk_05_l, loc=<7.522 0 0>, rot=<0 0 0 1>}
 dyn_Earring_l_MSTR {parent=head, loc=<0.241 0.2832 -6.3709>, rot=<0.0029 -0.7071 -0.0029 -0.7071>}
 dyn_Earring_01_l {parent=dyn_Earring_l_MSTR, loc=<0.0781 0.0486 -0.3084>, rot=<0.2701 -0.6521 -0.2711 -0.6544>}
 dyn_Sunglasses_MSTR {parent=head, loc=<9.9088 1.4816 0>, rot=<-0.5836 -0.3993 -0.3993 -0.5836>}
 dyn_Sunglasses_01_ct {parent=dyn_Sunglasses_MSTR, loc=<7.3929 0 0>, rot=<0 0 0 1>}
 dyn_Flower_01_l {parent=head, loc=<4.352 -0.5458 -6.7345>, rot=<-0.1206 -0.6967 -0.1264 -0.6957>}
 dyn_HairBraid_r_MSTR {parent=head, loc=<1.9594 -1.8872 6.5114>, rot=<-0.0469 -0.7055 -0.7055 -0.0469>}
 dyn_HairBraid_01_r {parent=dyn_HairBraid_r_MSTR, loc=<0 0 0>, rot=<0 0 0 1>}
 dyn_HairBraid_02_r {parent=dyn_HairBraid_01_r, loc=<6.406 0 0>, rot=<0 0.0886 0 -0.9961>}
 dyn_HairBraid_03_r {parent=dyn_HairBraid_02_r, loc=<5.432 0 0>, rot=<0 0.0835 0 -0.9965>}
 dyn_HairBraid_04_r {parent=dyn_HairBraid_03_r, loc=<5.328 0 0.0001>, rot=<0 0.0864 0 -0.9963>}
 dyn_HairBraid_05_r {parent=dyn_HairBraid_04_r, loc=<4.743 0 -0.0003>, rot=<0 0.0521 0 -0.9986>}
 dyn_HairBraid_06_r {parent=dyn_HairBraid_05_r, loc=<5.599 0 -0.0001>, rot=<0 0 0 1>}
 dyn_Hair_bk_r_MSTR {parent=head, loc=<9.6609 -1.634 6.406>, rot=<0.5 -0.5 0.5 -0.5>}
 dyn_Hair_bk_01_r {parent=dyn_Hair_bk_r_MSTR, loc=<0 0 0>, rot=<0 -0.4882 0 -0.8727>}
 dyn_Hair_bk_02_r {parent=dyn_Hair_bk_01_r, loc=<7.522 0 0>, rot=<0 -0.16 0 -0.9871>}
 dyn_Hair_bk_03_r {parent=dyn_Hair_bk_02_r, loc=<7.522 0 0>, rot=<0 0.1298 0 -0.9915>}
 dyn_Hair_bk_04_r {parent=dyn_Hair_bk_03_r, loc=<7.522 0 0>, rot=<0 -0.1239 0 -0.9923>}
 dyn_Hair_bk_05_r {parent=dyn_Hair_bk_04_r, loc=<7.522 0 0>, rot=<0 -0.1201 0 -0.9928>}
 dyn_Hair_bk_06_r {parent=dyn_Hair_bk_05_r, loc=<7.522 0 0>, rot=<0 0 0 -1>}
 dyn_Earring_r_MSTR {parent=head, loc=<0.2414 0.2832 6.3709>, rot=<0.7071 0.0029 0.7071 -0.0029>}
 dyn_Earring_01_r {parent=dyn_Earring_r_MSTR, loc=<0.078 -0.0486 -0.309>, rot=<0.2701 -0.6521 -0.2711 -0.6544>}
 clavicle_l {parent=spine_05, loc=<5.4831 -0.1053 -3.2946>, rot=<0.0465 -0.7289 -0.009 -0.6829>}
 upperarm_l {parent=clavicle_l, loc=<9.8658 -1.2136 -0.4105>, rot=<0.0164 -0.3245 -0.0269 -0.9454>}
 clavicle_r {parent=spine_05, loc=<5.4831 -0.1053 3.2946>, rot=<0.7289 0.0465 -0.6829 0.009>}
 upperarm_r {parent=clavicle_r, loc=<-9.8658 1.2136 0.4105>, rot=<0.0164 -0.3245 -0.0269 -0.9454>}
]
variables [
 Is Pistol Alpha {type=double, default=0}
 Is Rifle Alpha {type=double, default=0}
]
