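# Physics asset for the CubeCoast medium female hairpin face accessory, covering dangling earrings, two hair braids and a left-side flower.
# Capsules on head, neck, upper spine, clavicles and upper arms form the body shapes.
# Five-link braids and single-link earrings hang from zero-size master bodies constrained to the head, while the flower attaches directly.
# Dynamic links have limited swing and tight twist; earring-master and head-to-spine constraints leave rotation free.
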PhysicsAsset /BRCosmetics/Characters/Player/Female/Medium/Bodies/F_MED_CubeCoast/Meshes/Parts/F_MED_CubeCoast_FaceAcc_Hairpin_Physics
bodies {
 clavicle_l [{shapeType=Capsule, length=12, radius=5.25, shapeName=clavicle_l_capsule, center=<3.5 -1 0>, rotation=<rot 80 -0.0003 -0.0003>}]
 clavicle_r [{shapeType=Capsule, length=12, radius=5.25, shapeName=clavicle_r_capsule, center=<-3.5 1 0>, rotation=<rot 80 -0.0003 -0.0004>}]
 dyn_Earring_01_l [{shapeType=Capsule, length=1, radius=0.75, shapeName=dyn_Earring_01_l_capsule, center=<2.5 0 0>, rotation=<rot 90 0 0>}]
 dyn_Earring_01_r [{shapeType=Capsule, length=1, radius=0.75, shapeName=dyn_Earring_01_r_capsule, center=<2.5 0 0>, rotation=<rot 90 0 0>}]
 dyn_Earring_l_MSTR [{shapeType=Capsule, length=0, radius=0, shapeName=dyn_Earring_l_MSTR_capsule, rotation=<rot 90 0 0>}]
 dyn_Earring_r_MSTR [{shapeType=Capsule, length=0, radius=0, shapeName=dyn_Earring_r_MSTR_capsule, rotation=<rot 90 0 0>}]
 dyn_Flower_01_l [{shapeType=Capsule, length=0.5, radius=3.75, shapeName=dyn_Flower_01_l_capsule, center=<3.5 0 0>, rotation=<rot 90 0 0>}]
 dyn_HairBraid_01_l [{shapeType=Capsule, length=3, radius=1, shapeName=dyn_HairBraid_01_l_capsule, center=<2.75 0 0>, rotation=<rot 90 0 0>}]
 dyn_HairBraid_01_r [{shapeType=Capsule, length=3, radius=1, shapeName=dyn_HairBraid_01_r_capsule, center=<2.75 0 0>, rotation=<rot 90 0 0>}]
 dyn_HairBraid_02_l [{shapeType=Capsule, length=3, radius=1, shapeName=dyn_HairBraid_02_l_capsule, center=<2.75 0 0>, rotation=<rot 90 0 0>}]
 dyn_HairBraid_02_r [{shapeType=Capsule, length=3, radius=1, shapeName=dyn_HairBraid_02_r_capsule, center=<2.75 0 0>, rotation=<rot 90 0 0>}]
 dyn_HairBraid_03_l [{shapeType=Capsule, length=3, radius=1, shapeName=dyn_HairBraid_03_l_capsule, center=<2.75 0 0>, rotation=<rot 90 0 0>}]
 dyn_HairBraid_03_r [{shapeType=Capsule, length=3, radius=1, shapeName=dyn_HairBraid_03_r_capsule, center=<2.75 0 0>, rotation=<rot 90 0 0>}]
 dyn_HairBraid_04_l [{shapeType=Capsule, length=3, radius=1, shapeName=dyn_HairBraid_04_l_capsule, center=<2.75 0 0>, rotation=<rot 90 0 0>}]
 dyn_HairBraid_04_r [{shapeType=Capsule, length=3, radius=1, shapeName=dyn_HairBraid_04_r_capsule, center=<2.75 0 0>, rotation=<rot 90 0 0>}]
 dyn_HairBraid_05_l [{shapeType=Capsule, length=3, radius=1, shapeName=dyn_HairBraid_05_l_capsule, center=<2.75 0 0>, rotation=<rot 90 0 0>}]
 dyn_HairBraid_05_r [{shapeType=Capsule, length=3, radius=1, shapeName=dyn_HairBraid_05_r_capsule, center=<2.75 0 0>, rotation=<rot 90 0 0>}]
 dyn_HairBraid_l_MSTR [{shapeType=Capsule, length=0, radius=0, shapeName=dyn_HairBraid_l_MSTR_capsule, rotation=<rot 90 0 0>}]
 dyn_HairBraid_r_MSTR [{shapeType=Capsule, length=0, radius=0, shapeName=dyn_HairBraid_r_MSTR_capsule, rotation=<rot 90 0 0>}]
 head [{shapeType=Capsule, length=6, radius=6, shapeName=head_capsule, center=<3.5 3 0>, rotation=<rot 90 0 0>}]
 neck_01 [{shapeType=Capsule, length=8, radius=4.5, shapeName=neck_01_capsule, center=<3.6683 -0.2876 0>, rotation=<rot 90 0 0>}]
 spine_05 [
  {shapeType=Capsule, length=8, radius=9, shapeName=spine_05_capsule, center=<-3 1 0>}
  {shapeType=Capsule, length=8, radius=7, shapeName=spine_05_capsule, center=<-6 5 0>}
  {shapeType=Capsule, length=15, radius=8, shapeName=spine_05_capsule, center=<11.5207 -6.0296 0>}
 ]
 upperarm_l [{shapeType=Capsule, length=20, radius=3.5, shapeName=upperarm_l_capsule, center=<12 -0.5 0>, rotation=<rot 90 0 0>}]
 upperarm_r [{shapeType=Capsule, length=20, radius=3.5, shapeName=upperarm_r_capsule, center=<-12 0.5 0>, rotation=<rot 90 0 0>}]
}
constraints [
 {
  bone1Name=dyn_Earring_l_MSTR
  bone2Name=head
  swing1LimitDegrees=45
  swing1Motion=Free
  swing2LimitDegrees=45
  swing2Motion=Free
  twistLimitDegrees=45
  twistMotion=Free
 }
 {bone1Name=dyn_Earring_01_l, bone2Name=dyn_Earring_l_MSTR, swing1LimitDegrees=35, swing2LimitDegrees=35, twistLimitDegrees=2}
 {
  bone1Name=dyn_Earring_r_MSTR
  bone2Name=head
  swing1LimitDegrees=45
  swing1Motion=Free
  swing2LimitDegrees=45
  swing2Motion=Free
  twistLimitDegrees=45
  twistMotion=Free
 }
 {bone1Name=dyn_Earring_01_r, bone2Name=dyn_Earring_r_MSTR, swing1LimitDegrees=35, swing2LimitDegrees=35, twistLimitDegrees=2}
 {bone1Name=dyn_HairBraid_l_MSTR, bone2Name=head, swing1LimitDegrees=35, swing2LimitDegrees=35, twistLimitDegrees=2}
 {bone1Name=dyn_HairBraid_01_l, bone2Name=dyn_HairBraid_l_MSTR, swing1LimitDegrees=15, swing2LimitDegrees=25, twistLimitDegrees=2}
 {bone1Name=dyn_HairBraid_02_l, bone2Name=dyn_HairBraid_01_l, swing1LimitDegrees=15, swing2LimitDegrees=25, twistLimitDegrees=2}
 {bone1Name=dyn_HairBraid_03_l, bone2Name=dyn_HairBraid_02_l, swing1LimitDegrees=15, swing2LimitDegrees=25, twistLimitDegrees=2}
 {bone1Name=dyn_HairBraid_04_l, bone2Name=dyn_HairBraid_03_l, swing1LimitDegrees=15, swing2LimitDegrees=25, twistLimitDegrees=2}
 {bone1Name=dyn_HairBraid_05_l, bone2Name=dyn_HairBraid_04_l, swing1LimitDegrees=15, swing2LimitDegrees=25, twistLimitDegrees=2}
 {bone1Name=dyn_HairBraid_r_MSTR, bone2Name=head, swing1LimitDegrees=35, swing2LimitDegrees=35, twistLimitDegrees=2}
 {bone1Name=dyn_HairBraid_01_r, bone2Name=dyn_HairBraid_r_MSTR, swing1LimitDegrees=15, swing2LimitDegrees=25, twistLimitDegrees=2}
 {bone1Name=dyn_HairBraid_02_r, bone2Name=dyn_HairBraid_01_r, swing1LimitDegrees=15, swing2LimitDegrees=25, twistLimitDegrees=2}
 {bone1Name=dyn_HairBraid_03_r, bone2Name=dyn_HairBraid_02_r, swing1LimitDegrees=15, swing2LimitDegrees=25, twistLimitDegrees=2}
 {bone1Name=dyn_HairBraid_04_r, bone2Name=dyn_HairBraid_03_r, swing1LimitDegrees=15, swing2LimitDegrees=25, twistLimitDegrees=2}
 {bone1Name=dyn_HairBraid_05_r, bone2Name=dyn_HairBraid_04_r, swing1LimitDegrees=15, swing2LimitDegrees=25, twistLimitDegrees=2}
 {
  bone1Name=head
  bone2Name=spine_05
  swing1LimitDegrees=45
  swing1Motion=Free
  swing2LimitDegrees=45
  swing2Motion=Free
  twistLimitDegrees=45
  twistMotion=Free
 }
 {bone1Name=dyn_Flower_01_l, bone2Name=head, swing1LimitDegrees=25, swing2LimitDegrees=25, twistLimitDegrees=2}
]
